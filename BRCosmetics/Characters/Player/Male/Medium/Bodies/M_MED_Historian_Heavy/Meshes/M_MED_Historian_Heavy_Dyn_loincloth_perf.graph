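# Physics asset for the loincloth of the male medium Historian Heavy body, apparently a performance variant, with pelvis and thigh capsules as body shapes.
# Capsule chains form the back, back-side and center-front flaps, while boxes form the front-side flaps.
# Back chains hang from the belt and front chains from MSTR bones, with limited swing and tight twist.
# Angularly free constraints link belt and thighs to pelvis, and cross-link center-front segments to the right flap.
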PhysicsAsset /BRCosmetics/Characters/Player/Male/Medium/Bodies/M_MED_Historian_Heavy/Meshes/M_MED_Historian_Heavy_Dyn_loincloth_perf
bodies {
 dyn_belt [{shapeType=Capsule, length=0.505, radius=0.505}]
 dyn_loincloth_bk_01 [{shapeType=Capsule, length=4.4062, radius=10, center=<4.7651 0 10>, rotation=<rot 90 0 0>}]
 dyn_loincloth_bk_02 [{shapeType=Capsule, length=4.4062, radius=10, center=<4.7651 0 10>, rotation=<rot 90 0 0>}]
 dyn_loincloth_bk_03 [{shapeType=Capsule, length=4.4062, radius=10, center=<4.7651 0 10>, rotation=<rot 90 0 0>}]
 dyn_loincloth_bk_04 [{shapeType=Capsule, length=4.4062, radius=10, center=<4.7651 0 10>, rotation=<rot 90 0 0>}]
 dyn_loincloth_bk_l_01 [{shapeType=Capsule, length=4.4062, radius=10, center=<4.7651 0 10>, rotation=<rot 90 0 0>}]
 dyn_loincloth_bk_l_02 [{shapeType=Capsule, length=4.4062, radius=10, center=<4.7651 0 10>, rotation=<rot 90 0 0>}]
 dyn_loincloth_bk_r_01 [{shapeType=Capsule, length=5.4062, radius=9, center=<4.7652 0 8.6>, rotation=<rot 90 0 0>}]
 dyn_loincloth_bk_r_02 [{shapeType=Capsule, length=4.4062, radius=10, center=<4.7652 0 8.6>, rotation=<rot 90 0 0>}]
 dyn_loincloth_fr_01 [{shapeType=Capsule, length=5.0312, radius=5, center=<7.3671 0 5>, rotation=<rot 90 0 0>}]
 dyn_loincloth_fr_02 [{shapeType=Capsule, length=9.375, radius=5, center=<4.7651 0 5>, rotation=<rot 90 0 0>}]
 dyn_loincloth_fr_03 [{shapeType=Capsule, length=9.375, radius=5, center=<4.7651 0 5>, rotation=<rot 90 0 0>}]
 dyn_loincloth_fr_04 [{shapeType=Capsule, length=9.375, radius=5, center=<4.7651 0 5>, rotation=<rot 90 0 0>}]
 dyn_loincloth_fr_MSTR [{shapeType=Capsule, length=0.505, radius=0.505}]
 dyn_loincloth_fr_l_01 [{shapeType=Box, extentX=10.75, extentY=9.375, extentZ=2.5625, center=<5.1348 0 0>, rotation=<rot 0 0 -6.383>}]
 dyn_loincloth_fr_l_MSTR [{shapeType=Capsule, length=0.505, radius=0.505}]
 dyn_loincloth_fr_r_01 [{shapeType=Box, extentX=10.75, extentY=9.375, extentZ=5.0625, center=<5.1348 0 0>}]
 dyn_loincloth_fr_r_02 [{shapeType=Box, extentX=21.375, extentY=9.375, extentZ=5.0625, center=<10.7109 0 -0.0001>}]
 dyn_loincloth_fr_r_MSTR [{shapeType=Capsule, length=0.505, radius=0.505}]
 pelvis [{shapeType=Capsule, length=10, radius=13, center=<-2.5165 -1.8526 0.7152>, rotation=<rot 0 0 -0.9064>}]
 thigh_l [{shapeType=Capsule, length=20, radius=14, center=<-22.4447 -2.0308 2.7228>, rotation=<rot 90 -90 -92.1981>}]
 thigh_r [{shapeType=Capsule, length=20, radius=14, center=<23.7161 3.0948 -1.7609>, rotation=<rot 89.4539 -175.5101 -176.0004>}]
}
constraints [
 {bone1Name=dyn_loincloth_fr_l_01, bone2Name=dyn_loincloth_fr_l_MSTR, swing1LimitDegrees=15, swing2LimitDegrees=15, twistLimitDegrees=5}
 {bone1Name=dyn_loincloth_bk_01, bone2Name=dyn_belt, swing1LimitDegrees=45, swing2LimitDegrees=45, twistLimitDegrees=5}
 {bone1Name=dyn_loincloth_bk_02, bone2Name=dyn_loincloth_bk_01, swing1LimitDegrees=45, swing2LimitDegrees=45, twistLimitDegrees=5}
 {bone1Name=dyn_loincloth_bk_03, bone2Name=dyn_loincloth_bk_02, swing1LimitDegrees=45, swing2LimitDegrees=45, twistLimitDegrees=5}
 {bone1Name=dyn_loincloth_bk_04, bone2Name=dyn_loincloth_bk_03, swing1LimitDegrees=45, swing2LimitDegrees=45, twistLimitDegrees=5}
 {bone1Name=dyn_loincloth_bk_l_01, bone2Name=dyn_belt, swing1LimitDegrees=25, swing2LimitDegrees=45, twistLimitDegrees=5}
 {bone1Name=dyn_loincloth_bk_l_02, bone2Name=dyn_loincloth_bk_l_01, swing1LimitDegrees=45, swing2LimitDegrees=45, twistLimitDegrees=5}
 {bone1Name=dyn_loincloth_bk_r_01, bone2Name=dyn_belt, swing1LimitDegrees=25, swing2LimitDegrees=45, twistLimitDegrees=5}
 {bone1Name=dyn_loincloth_bk_r_02, bone2Name=dyn_loincloth_bk_r_01, swing1LimitDegrees=45, swing2LimitDegrees=45, twistLimitDegrees=5}
 {bone1Name=dyn_loincloth_fr_01, bone2Name=dyn_loincloth_fr_MSTR, swing1LimitDegrees=10, swing2LimitDegrees=10, twistLimitDegrees=5}
 {bone1Name=dyn_loincloth_fr_02, bone2Name=dyn_loincloth_fr_01, swing1LimitDegrees=25, swing2LimitDegrees=25, twistLimitDegrees=5}
 {bone1Name=dyn_loincloth_fr_03, bone2Name=dyn_loincloth_fr_02, swing1LimitDegrees=45, swing2LimitDegrees=45, twistLimitDegrees=5}
 {bone1Name=dyn_loincloth_fr_04, bone2Name=dyn_loincloth_fr_03, swing1LimitDegrees=45, swing2LimitDegrees=45, twistLimitDegrees=5}
 {bone1Name=dyn_loincloth_fr_r_01, bone2Name=dyn_loincloth_fr_r_MSTR, swing1LimitDegrees=15, swing2LimitDegrees=15, twistLimitDegrees=5}
 {bone1Name=dyn_loincloth_fr_r_02, bone2Name=dyn_loincloth_fr_r_01, swing1LimitDegrees=15, swing2LimitDegrees=15, twistLimitDegrees=5}
 {
  bone1Name=dyn_belt
  bone2Name=pelvis
  swing1LimitDegrees=45
  swing1Motion=Free
  swing2LimitDegrees=45
  swing2Motion=Free
  twistLimitDegrees=45
  twistMotion=Free
 }
 {
  bone1Name=thigh_l
  bone2Name=pelvis
  swing1LimitDegrees=45
  swing1Motion=Free
  swing2LimitDegrees=45
  swing2Motion=Free
  twistLimitDegrees=45
  twistMotion=Free
 }
 {
  bone1Name=thigh_r
  bone2Name=pelvis
  swing1LimitDegrees=45
  swing1Motion=Free
  swing2LimitDegrees=45
  swing2Motion=Free
  twistLimitDegrees=45
  twistMotion=Free
 }
 {
  bone1Name=dyn_loincloth_fr_02
  bone2Name=dyn_loincloth_fr_r_02
  swing1LimitDegrees=45
  swing1Motion=Free
  swing2LimitDegrees=45
  swing2Motion=Free
  twistLimitDegrees=45
  twistMotion=Free
 }
 {
  bone1Name=dyn_loincloth_fr_03
  bone2Name=dyn_loincloth_fr_r_02
  swing1LimitDegrees=45
  swing1Motion=Free
  swing2LimitDegrees=45
  swing2Motion=Free
  twistLimitDegrees=45
  twistMotion=Free
 }
]
