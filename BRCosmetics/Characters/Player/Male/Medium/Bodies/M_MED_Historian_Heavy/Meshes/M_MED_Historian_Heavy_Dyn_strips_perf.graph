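# Physics asset for the dynamic cloth strips on the male medium Historian Heavy player body, a variant suffixed "perf".
# Each lettered left or right strip is a short chain of capsule bodies hanging from a tiny MSTR root body, linked by swing and twist limited constraints.
# The left B and right A roots join front loincloth bodies with free angular motion.
# Two large thigh capsules provide leg volumes alongside the strips.
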PhysicsAsset /BRCosmetics/Characters/Player/Male/Medium/Bodies/M_MED_Historian_Heavy/Meshes/M_MED_Historian_Heavy_Dyn_strips_perf
bodies {
 dyn_loincloth_fr_l_02 [{
   shapeType=Box
   extentX=6.3987
   extentY=3.2518
   extentZ=9.3421
   shapeName=dyn_loincloth_fr_l_02_box
   center=<2.186 -2.322 -1.1056>
   rotation=<rot 0 -84.8054 -85.3805>
  }]
 dyn_loincloth_fr_r_01 [{
   shapeType=Capsule
   length=7.1746
   radius=3.6406
   shapeName=dyn_loincloth_fr_r_01_capsule
   center=<3.2637 4.9898 -2.0713>
   rotation=<rot 0 67.5145 -80.3132>
  }]
 dyn_strip_l_B_01 [{shapeType=Capsule, length=0.5938, radius=4.25, center=<6.2545 -0.0004 5.2977>, rotation=<rot 90 0 0>}]
 dyn_strip_l_B_02 [{shapeType=Capsule, length=6.0312, radius=4.25, center=<5.2586 -0.0005 5.2295>, rotation=<rot 90 0 0>}]
 dyn_strip_l_B_MSTR [{shapeType=Capsule, length=0.505, radius=0.505, shapeName=dyn_strip_l_B_MSTR_capsule, rotation=<rot 0 -90 0>}]
 dyn_strip_l_D_01 [{shapeType=Capsule, length=6.25, radius=4.4688, center=<5.2585 -0.0004 4.1656>, rotation=<rot 90 0 0>}]
 dyn_strip_l_D_02 [{shapeType=Capsule, length=6.25, radius=4.4688, center=<5.2585 -0.0004 4.1656>, rotation=<rot 90 0 0>}]
 dyn_strip_l_D_MSTR [{shapeType=Capsule, length=0.505, radius=0.505}]
 dyn_strip_l_E_01 [{shapeType=Capsule, length=6.0938, radius=4.5, center=<5.2585 -0.0004 4.0819>, rotation=<rot 90 0 0>}]
 dyn_strip_l_E_02 [{shapeType=Capsule, length=6.0938, radius=4.5, center=<5.2585 -0.0004 4.0819>, rotation=<rot 90 0 0>}]
 dyn_strip_l_E_03 [{shapeType=Capsule, length=6.0938, radius=4.5, center=<5.2585 -0.0004 4.0819>, rotation=<rot 90 0 0>}]
 dyn_strip_l_E_MSTR [{shapeType=Capsule, length=0.505, radius=0.505}]
 dyn_strip_l_F_01 [{shapeType=Capsule, length=6.0938, radius=4.5, center=<5.2585 -0.0004 4.2063>, rotation=<rot 90 0 0>}]
 dyn_strip_l_F_02 [{shapeType=Capsule, length=6.0938, radius=4.5, center=<5.2585 -0.0004 4.2063>, rotation=<rot 90 0 0>}]
 dyn_strip_l_F_MSTR [{shapeType=Capsule, length=0.505, radius=0.505}]
 dyn_strip_l_G_01 [{shapeType=Capsule, length=5.3438, radius=5.3125, center=<5.2586 -0.0004 4.5>, rotation=<rot 90 0 0>}]
 dyn_strip_l_G_02 [{shapeType=Capsule, length=5.3438, radius=5.3125, center=<5.2586 -0.0004 4.5>, rotation=<rot 90 0 0>}]
 dyn_strip_l_G_03 [{shapeType=Capsule, length=5.3438, radius=5.3125, center=<5.2586 -0.0004 4.5>, rotation=<rot 90 0 0>}]
 dyn_strip_l_G_MSTR [{shapeType=Capsule, length=0.505, radius=0.505}]
 dyn_strip_l_H_01 [{shapeType=Capsule, length=5.6875, radius=4.9688, center=<5.2585 0 4.5>, rotation=<rot 90 0 0>}]
 dyn_strip_l_H_02 [{shapeType=Capsule, length=5.3125, radius=5.3438, center=<5.2585 0 4.5>, rotation=<rot 90 0 0>}]
 dyn_strip_l_H_03 [{shapeType=Capsule, length=5.3125, radius=5.3438, center=<5.2585 0 4.5>, rotation=<rot 90 0 0>}]
 dyn_strip_l_H_MSTR [{shapeType=Capsule, length=0.505, radius=0.505}]
 dyn_strip_l_J_01 [{shapeType=Capsule, length=6.4062, radius=4.5, center=<5.2586 -0.0004 4.5>, rotation=<rot 90 0 0>}]
 dyn_strip_l_J_02 [{shapeType=Capsule, length=5.9688, radius=4.9375, center=<5.2586 -0.0004 4.5>, rotation=<rot 90 0 0>}]
 dyn_strip_l_J_03 [{shapeType=Capsule, length=5.9688, radius=4.9375, center=<5.2586 -0.0004 4.5>, rotation=<rot 90 0 0>}]
 dyn_strip_l_J_MSTR [{shapeType=Capsule, length=0.505, radius=0.505}]
 dyn_strip_l_K_01 [{shapeType=Capsule, length=6.1875, radius=4.5, center=<5.2586 -0.0004 4.5>, rotation=<rot 90 0 0>}]
 dyn_strip_l_K_02 [{shapeType=Capsule, length=6.1875, radius=4.5, center=<5.2586 -0.0004 4.5>, rotation=<rot 90 0 0>}]
 dyn_strip_l_K_MSTR [{shapeType=Capsule, length=0.505, radius=0.505}]
 dyn_strip_l_L_01 [{shapeType=Capsule, length=6.1562, radius=4.5, center=<5.2586 -0.0004 4.5>, rotation=<rot 90 0 0>}]
 dyn_strip_l_L_MSTR [{shapeType=Capsule, length=0.505, radius=0.505}]
 dyn_strip_r_A_01 [{shapeType=Capsule, length=1.3438, radius=4, center=<6.9318 -0.0004 3.9158>, rotation=<rot 90 0 0>}]
 dyn_strip_r_A_02 [{shapeType=Capsule, length=6.75, radius=4, center=<5.2579 -0.0005 3.569>, rotation=<rot 90 0 0>}]
 dyn_strip_r_A_MSTR [{shapeType=Capsule, length=0.505, radius=0.505}]
 dyn_strip_r_B_01 [{shapeType=Capsule, length=5.1562, radius=4.75, center=<5.2586 -0.0004 3.199>, rotation=<rot 90 0 0>}]
 dyn_strip_r_B_02 [{shapeType=Capsule, length=5.1562, radius=4.75, center=<5.2586 -0.0004 3.199>, rotation=<rot 90 0 0>}]
 dyn_strip_r_B_03 [{shapeType=Capsule, length=5.1562, radius=4.75, center=<5.2586 -0.0004 3.199>, rotation=<rot 90 0 0>}]
 dyn_strip_r_B_MSTR [{shapeType=Capsule, length=0.505, radius=0.505}]
 dyn_strip_r_C_01 [{shapeType=Capsule, length=5.25, radius=4.625, center=<5.2584 -0.0004 3.0926>, rotation=<rot 90 0 0>}]
 dyn_strip_r_C_02 [{shapeType=Capsule, length=5.25, radius=4.625, center=<5.2584 -0.0004 3.0926>, rotation=<rot 90 0 0>}]
 dyn_strip_r_C_MSTR [{shapeType=Capsule, length=0.505, radius=0.505}]
 dyn_strip_r_D_01 [{shapeType=Capsule, length=6.5625, radius=4.5, center=<5.2587 -0.0004 3.4428>, rotation=<rot 90 0 0>}]
 dyn_strip_r_D_02 [{shapeType=Capsule, length=5.7812, radius=4.5, center=<5.2587 -0.0004 3.4428>, rotation=<rot 90 0 0>}]
 dyn_strip_r_D_MSTR [{shapeType=Capsule, length=0.505, radius=0.505}]
 dyn_strip_r_E_01 [{shapeType=Capsule, length=5.5, radius=4.5, center=<5.2586 -0.0004 4.5>, rotation=<rot 90 0 0>}]
 dyn_strip_r_E_02 [{shapeType=Capsule, length=4.9375, radius=5.0625, center=<5.2586 -0.0004 4.5>, rotation=<rot 90 0 0>}]
 dyn_strip_r_E_MSTR [{shapeType=Capsule, length=0.505, radius=0.505}]
 dyn_strip_r_F_01 [{shapeType=Capsule, length=6.1875, radius=4.5, center=<5.2586 -0.0004 4.5>, rotation=<rot 90 0 0>}]
 dyn_strip_r_F_02 [{shapeType=Capsule, length=6.1875, radius=4.5, center=<5.2586 -0.0004 4.5>, rotation=<rot 90 0 0>}]
 dyn_strip_r_F_MSTR [{shapeType=Capsule, length=0.505, radius=0.505}]
 dyn_strip_r_G_01 [{shapeType=Capsule, length=6.3125, radius=4.5, center=<5.2586 -0.0004 4.5>, rotation=<rot 90 0 0>}]
 dyn_strip_r_G_02 [{shapeType=Capsule, length=6.3125, radius=4.5, center=<5.2586 -0.0004 4.5>, rotation=<rot 90 0 0>}]
 dyn_strip_r_G_MSTR [{shapeType=Capsule, length=0.505, radius=0.505}]
 dyn_strip_r_H_01 [{shapeType=Capsule, length=6.3125, radius=4.5, center=<5.2586 -0.0004 4.5>, rotation=<rot 90 0 0>}]
 dyn_strip_r_H_MSTR [{shapeType=Capsule, length=0.505, radius=0.505}]
 thigh_l [{shapeType=Capsule, length=19.0312, radius=13.2, center=<-19.8857 -2.6068 2.8181>, rotation=<rot 90 -90 -92.1981>}]
 thigh_r [{shapeType=Capsule, length=19.0312, radius=13.2, center=<19.8857 2.6069 -2.574>, rotation=<rot 90 34.5085 32.3104>}]
}
constraints [
 {bone1Name=dyn_strip_l_D_01, bone2Name=dyn_strip_l_D_MSTR, swing1LimitDegrees=45, swing2LimitDegrees=45, twistLimitDegrees=10}
 {bone1Name=dyn_strip_l_D_02, bone2Name=dyn_strip_l_D_01, swing1LimitDegrees=45, swing2LimitDegrees=45, twistLimitDegrees=10}
 {bone1Name=dyn_strip_l_E_01, bone2Name=dyn_strip_l_E_MSTR, swing1LimitDegrees=35, swing2LimitDegrees=35, twistLimitDegrees=10}
 {bone1Name=dyn_strip_l_E_02, bone2Name=dyn_strip_l_E_01, swing1LimitDegrees=35, swing2LimitDegrees=35, twistLimitDegrees=10}
 {bone1Name=dyn_strip_l_E_03, bone2Name=dyn_strip_l_E_02, swing1LimitDegrees=35, swing2LimitDegrees=35, twistLimitDegrees=5}
 {bone1Name=dyn_strip_l_F_01, bone2Name=dyn_strip_l_F_MSTR, swing1LimitDegrees=20, swing2LimitDegrees=20, twistLimitDegrees=5}
 {bone1Name=dyn_strip_l_F_02, bone2Name=dyn_strip_l_F_01, swing1LimitDegrees=20, swing2LimitDegrees=20, twistLimitDegrees=5}
 {bone1Name=dyn_strip_l_G_01, bone2Name=dyn_strip_l_G_MSTR, swing1LimitDegrees=20, swing2LimitDegrees=20, twistLimitDegrees=5}
 {bone1Name=dyn_strip_l_G_02, bone2Name=dyn_strip_l_G_01, swing1LimitDegrees=20, swing2LimitDegrees=20, twistLimitDegrees=5}
 {bone1Name=dyn_strip_l_G_03, bone2Name=dyn_strip_l_G_02, swing1LimitDegrees=20, swing2LimitDegrees=20, twistLimitDegrees=5}
 {bone1Name=dyn_strip_l_H_01, bone2Name=dyn_strip_l_H_MSTR, swing1LimitDegrees=20, swing2LimitDegrees=20, twistLimitDegrees=15}
 {bone1Name=dyn_strip_l_H_02, bone2Name=dyn_strip_l_H_01, swing1LimitDegrees=20, swing2LimitDegrees=20, twistLimitDegrees=15}
 {bone1Name=dyn_strip_l_H_03, bone2Name=dyn_strip_l_H_02, swing1LimitDegrees=20, swing2LimitDegrees=20, twistLimitDegrees=15}
 {bone1Name=dyn_strip_l_J_01, bone2Name=dyn_strip_l_J_MSTR, swing1LimitDegrees=20, swing2LimitDegrees=20, twistLimitDegrees=15}
 {bone1Name=dyn_strip_l_J_02, bone2Name=dyn_strip_l_J_01, swing1LimitDegrees=20, swing2LimitDegrees=20, twistLimitDegrees=15}
 {bone1Name=dyn_strip_l_J_03, bone2Name=dyn_strip_l_J_02, swing1LimitDegrees=20, swing2LimitDegrees=20, twistLimitDegrees=15}
 {bone1Name=dyn_strip_l_K_01, bone2Name=dyn_strip_l_K_MSTR, swing1LimitDegrees=20, swing2LimitDegrees=20, twistLimitDegrees=5}
 {bone1Name=dyn_strip_l_K_02, bone2Name=dyn_strip_l_K_01, swing1LimitDegrees=20, swing2LimitDegrees=20, twistLimitDegrees=5}
 {bone1Name=dyn_strip_l_L_01, bone2Name=dyn_strip_l_L_MSTR, swing1LimitDegrees=20, swing2LimitDegrees=20, twistLimitDegrees=5}
 {bone1Name=dyn_strip_r_A_01, bone2Name=dyn_strip_r_A_MSTR, swing1LimitDegrees=5, swing2LimitDegrees=5, twistLimitDegrees=5}
 {bone1Name=dyn_strip_r_A_02, bone2Name=dyn_strip_r_A_01, swing1LimitDegrees=20, swing2LimitDegrees=20, twistLimitDegrees=5}
 {bone1Name=dyn_strip_r_B_01, bone2Name=dyn_strip_r_B_MSTR, swing1LimitDegrees=20, swing2LimitDegrees=20, twistLimitDegrees=5}
 {bone1Name=dyn_strip_r_B_02, bone2Name=dyn_strip_r_B_01, swing1LimitDegrees=20, swing2LimitDegrees=20, twistLimitDegrees=5}
 {bone1Name=dyn_strip_r_B_03, bone2Name=dyn_strip_r_B_02, swing1LimitDegrees=20, swing2LimitDegrees=20, twistLimitDegrees=5}
 {bone1Name=dyn_strip_r_C_01, bone2Name=dyn_strip_r_C_MSTR, swing1LimitDegrees=20, swing2LimitDegrees=20, twistLimitDegrees=5}
 {bone1Name=dyn_strip_r_D_01, bone2Name=dyn_strip_r_D_MSTR, swing1LimitDegrees=20, swing2LimitDegrees=20, twistLimitDegrees=5}
 {bone1Name=dyn_strip_r_D_02, bone2Name=dyn_strip_r_D_01, swing1LimitDegrees=20, swing2LimitDegrees=20, twistLimitDegrees=5}
 {bone1Name=dyn_strip_r_E_01, bone2Name=dyn_strip_r_E_MSTR, swing1LimitDegrees=20, swing2LimitDegrees=20, twistLimitDegrees=5}
 {bone1Name=dyn_strip_r_E_02, bone2Name=dyn_strip_r_E_01, swing1LimitDegrees=20, swing2LimitDegrees=20, twistLimitDegrees=5}
 {bone1Name=dyn_strip_r_F_01, bone2Name=dyn_strip_r_F_MSTR, swing1LimitDegrees=20, swing2LimitDegrees=20, twistLimitDegrees=2}
 {bone1Name=dyn_strip_r_G_01, bone2Name=dyn_strip_r_G_MSTR, swing1LimitDegrees=20, swing2LimitDegrees=20, twistLimitDegrees=2}
 {bone1Name=dyn_strip_r_C_02, bone2Name=dyn_strip_r_C_01, swing1LimitDegrees=20, swing2LimitDegrees=20, twistLimitDegrees=5}
 {bone1Name=dyn_strip_r_F_02, bone2Name=dyn_strip_r_F_01, swing1LimitDegrees=20, swing2LimitDegrees=20, twistLimitDegrees=5}
 {bone1Name=dyn_strip_r_G_02, bone2Name=dyn_strip_r_G_01, swing1LimitDegrees=20, swing2LimitDegrees=20, twistLimitDegrees=5}
 {bone1Name=dyn_strip_r_H_01, bone2Name=dyn_strip_r_H_MSTR, swing1LimitDegrees=20, swing2LimitDegrees=20, twistLimitDegrees=5}
 {
  bone1Name=dyn_strip_r_A_MSTR
  bone2Name=dyn_loincloth_fr_r_01
  swing1LimitDegrees=45
  swing1Motion=Free
  swing2LimitDegrees=45
  swing2Motion=Free
  twistLimitDegrees=45
  twistMotion=Free
 }
 {bone1Name=dyn_strip_l_B_01, bone2Name=dyn_strip_l_B_MSTR, swing1LimitDegrees=45, swing2LimitDegrees=45, twistLimitDegrees=5}
 {bone1Name=dyn_strip_l_B_02, bone2Name=dyn_strip_l_B_01, swing1LimitDegrees=45, swing2LimitDegrees=45, twistLimitDegrees=5}
 {
  bone1Name=dyn_strip_l_B_MSTR
  bone2Name=dyn_loincloth_fr_l_02
  swing1LimitDegrees=45
  swing1Motion=Free
  swing2LimitDegrees=45
  swing2Motion=Free
  twistLimitDegrees=45
  twistMotion=Free
 }
]
