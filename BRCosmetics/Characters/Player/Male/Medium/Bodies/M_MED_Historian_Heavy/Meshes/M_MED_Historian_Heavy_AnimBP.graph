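# Animation Blueprint for the male medium Historian Heavy cosmetic body, adding deformation and secondary motion to a pose copied from the attached parent mesh.
# Control Rigs handle twists, sculpters and deformers, then weighted constraints position gear, straps and shoulder pads.
# Look-At nodes and upper-arm Pose Drivers shape the pads, with additive offsets for skydiving and arm poses.
# RigidBody and AnimDynamics simulate loincloth and accessories, using a simulation-space root bone and gravity overrides.
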AnimBlueprint /BRCosmetics/Characters/Player/Male/Medium/Bodies/M_MED_Historian_Heavy/Meshes/M_MED_Historian_Heavy_AnimBP
target_skeleton=&/M_MED_Historian_Heavy_Skeleton
graphs [
 AnimGraph {nodes [
   n1 = "Constraint - dyn_shoulderpad_r_01_front_MSTR (PPPPPPP)"!(
    ComponentPose
    Alpha=1
    ConstraintWeights_0=0.0501
    ConstraintWeights_1:float=0.2684
    ConstraintWeights_2:float=0.1718
    ConstraintWeights_3:float=0.0321
    ConstraintWeights_4:float=0.0901
    ConstraintWeights_5:float=0.181
    ConstraintWeights_6:float=0.2065
   ) -> (Pose) {
    alphaBoolBlend {blendOption=Linear}
    boneToModify {boneName=dyn_shoulderpad_r_01_front_MSTR}
    constraintSetup [
     {targetBone {boneName=upperarm_twist_01_r}}
     {targetBone {boneName=deform_pectoral_fr_r}}
     {targetBone {boneName=deform_deltoid_fr_r}}
     {targetBone {boneName=spine_03}}
     {targetBone {boneName=spine_04}}
     {targetBone {boneName=spine_05}}
     {targetBone {boneName=clavicle_r}}
    ]
    constraintWeights [1,1,1,1,1,1,1]
   }
   n2 = "Constraint - dyn_shoulderpad_l_01_front_MSTR (PPPPPPP)"!(
    ComponentPose=@n1
    Alpha=1
    ConstraintWeights_0=0.2065
    ConstraintWeights_1:float=0.1718
    ConstraintWeights_2:float=0.2684
    ConstraintWeights_3:float=0.0321
    ConstraintWeights_4:float=0.0901
    ConstraintWeights_5:float=0.181
    ConstraintWeights_6:float=0.0501
   ) -> (Pose) {
    alphaBoolBlend {blendOption=Linear}
    boneToModify {boneName=dyn_shoulderpad_l_01_front_MSTR}
    constraintSetup [
     {targetBone {boneName=clavicle_l}}
     {targetBone {boneName=deform_deltoid_fr_l}}
     {targetBone {boneName=deform_pectoral_fr_l}}
     {targetBone {boneName=spine_03}}
     {targetBone {boneName=spine_04}}
     {targetBone {boneName=spine_05}}
     {targetBone {boneName=upperarm_twist_01_l}}
    ]
    constraintWeights [1,1,1,1,1,1,1]
   }
   n3 = "Constraint - dyn_shoulderpad_l_01_side_MSTR (PPPPPPPP)"!(
    ComponentPose=@n2
    Alpha=1
    ConstraintWeights_0=0.1339
    ConstraintWeights_1:float=0.2649
    ConstraintWeights_2:float=0.1596
    ConstraintWeights_3:float=0.1571
    ConstraintWeights_4:float=0.0233
    ConstraintWeights_5:float=0.0472
    ConstraintWeights_6:float=0.1266
    ConstraintWeights_7:float=0.0874
   ) -> (Pose) {
    alphaBoolBlend {blendOption=Linear}
    boneToModify {boneName=dyn_shoulderpad_l_01_side_MSTR}
    constraintSetup [
     {targetBone {boneName=clavicle_l}}
     {targetBone {boneName=dyn_shoulderpad_l_01_side}}
     {targetBone {boneName=deform_deltoid_fr_l}}
     {targetBone {boneName=deform_pectoral_fr_l}}
     {targetBone {boneName=spine_04}}
     {targetBone {boneName=spine_05}}
     {targetBone {boneName=upperarm_twist_01_l}}
     {targetBone {boneName=dyn_shoulderpad_l_01_front}}
    ]
    constraintWeights [1,1,1,1,1,1,1,1]
   }
   "Constraint - dyn_shoulderpad_r_01_side_MSTR (PPPPPPPP)"!(
    ComponentPose=@n3
    Alpha=1
    ConstraintWeights_0=0.0874
    ConstraintWeights_1:float=0.1266
    ConstraintWeights_2:float=0.1571
    ConstraintWeights_3:float=0.1596
    ConstraintWeights_4:float=0.0233
    ConstraintWeights_5:float=0.0472
    ConstraintWeights_6:float=0.1339
    ConstraintWeights_7:float=0.2649
   ) -> (Pose) {
    alphaBoolBlend {blendOption=Linear}
    boneToModify {boneName=dyn_shoulderpad_r_01_side_MSTR}
    constraintSetup [
     {targetBone {boneName=dyn_shoulderpad_r_01_front}}
     {targetBone {boneName=upperarm_twist_01_r}}
     {targetBone {boneName=deform_pectoral_fr_r}}
     {targetBone {boneName=deform_deltoid_fr_r}}
     {targetBone {boneName=spine_04}}
     {targetBone {boneName=spine_05}}
     {targetBone {boneName=clavicle_r}}
     {targetBone {boneName=dyn_shoulderpad_r_01_side}}
    ]
    constraintWeights [1,1,1,1,1,1,1,1]
   }
   n4 = Get Jump Alpha() -> (Jump Alpha:double)
   "Constraint - dyn_strip_r_B_MSTR (P)"(Alpha=@n4, ConstraintWeights_0=1) {
    alphaBoolBlend {blendOption=Linear}
    boneToModify {boneName=dyn_strip_r_B_MSTR}
    constraintSetup [{targetBone {boneName=thigh_r}}]
    constraintWeights [1]
   }
   n5 = Control Rig(Source=Component To Local()) {
    alphaBoolBlend {blendOption=Linear}
    controlRigAssetReference {blueprintRigClass=&/M_MED_Historian_Heavy_Deform_CtrlRig_shoulders_l.M_MED_Historian_Heavy_Deform_CtrlRig_shoulders_l_C}
   }
   Local To Component(LocalPose=@n5)
   n6 = Get GravityOverride()
   RigidBody!(ComponentPose, Alpha, OverrideWorldGravity:Vector=@n6, ExternalForce) -> (Pose) {
    bClampLinearTranslationLimitToRefPose=true
    bOverrideWorldGravity=true
    overridePhysicsAsset=&/M_MED_Historian_Heavy_Dyn_torso
    simulationSpace=BaseBoneSpace
    alphaBoolBlend {blendOption=Linear}
    baseBoneRef {boneName=dyn_root_simspace}
    componentAppliedLinearAccClamp=<3000 3000 3000>
    componentLinearAccScale {z=0.5}
    componentLinearVelScale {z=0.5}
    overrideWorldGravity {z=-980}
    simSpaceSettings {
     maxAngularAcceleration=1000
     maxAngularVelocity=3
     maxLinearAcceleration=250
     maxLinearVelocity=500
     velocityScaleZ=0
     worldAlpha=0.5
     externalLinearDragV=<0.5 0.5>
    }
   }
   n7 = Control Rig(Source=Copy Pose From Mesh() {bUseAttachedParent=true}) {
    alphaBoolBlend {blendOption=Linear}
    controlRigAssetReference {blueprintRigClass=&/M_MED_Historian_Heavy_Deform_CtrlRig_twists.M_MED_Historian_Heavy_Deform_CtrlRig_twists_C}
   }
   n8 = Control Rig(Source=@n7) {
    alphaBoolBlend {blendOption=Linear}
    controlRigAssetReference {blueprintRigClass=&/M_MED_Historian_Heavy_Deform_CtrlRig_sculpters.M_MED_Historian_Heavy_Deform_CtrlRig_sculpters_C}
   }
   n9 = Control Rig(Source=@n8) {
    alphaBoolBlend {blendOption=Linear}
    controlRigAssetReference {blueprintRigClass=&/M_MED_Historian_Heavy_Deform_CtrlRig_deformers.M_MED_Historian_Heavy_Deform_CtrlRig_deformers_C}
   }
   n10 = "Constraint - dyn_omega_01_MSTR (PPPP)"!(
    ComponentPose=Local To Component(LocalPose=@n9)
    Alpha=1
    ConstraintWeights_0=0.2683
    ConstraintWeights_1:float=0.1665
    ConstraintWeights_2:float=0.1775
    ConstraintWeights_3:float=0.3877
   ) -> (Pose) {
    alphaBoolBlend {blendOption=Linear}
    boneToModify {boneName=dyn_omega_01_MSTR}
    constraintSetup [{targetBone {boneName=clavicle_r}}, {targetBone {boneName=spine_03}}, {targetBone {boneName=spine_04}}, {targetBone {boneName=spine_05}}]
    constraintWeights [1,1,1,1]
   }
   n11 = "Constraint - dyn_satchel_l_MSTR (PPPP)"!(
    ComponentPose=@n10
    Alpha=1
    ConstraintWeights_0=0.2733
    ConstraintWeights_1:float=0.042
    ConstraintWeights_2:float=0.0328
    ConstraintWeights_3:float=0.6519
   ) -> (Pose) {
    alphaBoolBlend {blendOption=Linear}
    boneToModify {boneName=dyn_satchel_l_MSTR}
    constraintSetup [{targetBone {boneName=spine_01}}, {targetBone {boneName=spine_02}}, {targetBone {boneName=spine_03}}, {targetBone {boneName=pelvis}}]
    constraintWeights [1,1,1,1]
   }
   n12 = "Constraint - dyn_knife_l_MSTR (PPPP)"!(
    ComponentPose=@n11
    Alpha=1
    ConstraintWeights_0=0.3514
    ConstraintWeights_1:float=0.0892
    ConstraintWeights_2:float=0.0284
    ConstraintWeights_3:float=0.5311
   ) -> (Pose) {
    alphaBoolBlend {blendOption=Linear}
    boneToModify {boneName=dyn_knife_l_MSTR}
    constraintSetup [{targetBone {boneName=spine_01}}, {targetBone {boneName=spine_02}}, {targetBone {boneName=spine_03}}, {targetBone {boneName=pelvis}}]
    constraintWeights [1,1,1,1]
   }
   n13 = "Constraint - dyn_bag_l_MSTR (PPPPP)"!(
    ComponentPose=@n12
    Alpha=1
    ConstraintWeights_0=0.5397
    ConstraintWeights_1:float=0.012
    ConstraintWeights_2:float=0.0742
    ConstraintWeights_3:float=0.026
    ConstraintWeights_4:float=0.3482
   ) -> (Pose) {
    alphaBoolBlend {blendOption=Linear}
    boneToModify {boneName=dyn_bag_l_MSTR}
    constraintSetup [
     {targetBone {boneName=pelvis}}
     {targetBone {boneName=deform_thigh_fr_l}}
     {targetBone {boneName=spine_02}}
     {targetBone {boneName=spine_03}}
     {targetBone {boneName=spine_01}}
    ]
    constraintWeights [1,1,1,1,1]
   }
   n14 = "Constraint - dyn_belt_tips_top_MSTR (PPPP)"!(
    ComponentPose=@n13
    Alpha=1
    ConstraintWeights_0=0.3681
    ConstraintWeights_1:float=0.1817
    ConstraintWeights_2:float=0.0979
    ConstraintWeights_3:float=0.3523
   ) -> (Pose) {
    alphaBoolBlend {blendOption=Linear}
    boneToModify {boneName=dyn_belt_tips_top_MSTR}
    constraintSetup [{targetBone {boneName=spine_01}}, {targetBone {boneName=spine_02}}, {targetBone {boneName=spine_03}}, {targetBone {boneName=pelvis}}]
    constraintWeights [1,1,1,1]
   }
   n15 = "Constraint - dyn_belt_tips_bott_MSTR (PPPPP)"!(
    ComponentPose=@n14
    Alpha=1
    ConstraintWeights_0=0.3836
    ConstraintWeights_1:float=0.0624
    ConstraintWeights_2:float=0.0207
    ConstraintWeights_3:float=0.0244
    ConstraintWeights_4:float=0.5089
   ) -> (Pose) {
    alphaBoolBlend {blendOption=Linear}
    boneToModify {boneName=dyn_belt_tips_bott_MSTR}
    constraintSetup [
     {targetBone {boneName=spine_01}}
     {targetBone {boneName=spine_02}}
     {targetBone {boneName=spine_03}}
     {targetBone {boneName=deform_thigh_fr_l}}
     {targetBone {boneName=pelvis}}
    ]
    constraintWeights [1,1,1,1,1]
   }
   n16 = "Constraint - dyn_belly_medallion_l_MSTR (PPPPP)"!(
    ComponentPose=@n15
    Alpha=1
    ConstraintWeights_0=0.3888
    ConstraintWeights_1:float=0.1046
    ConstraintWeights_2:float=0.0427
    ConstraintWeights_3:float=0.0155
    ConstraintWeights_4:float=0.4483
   ) -> (Pose) {
    alphaBoolBlend {blendOption=Linear}
    boneToModify {boneName=dyn_belly_medallion_l_MSTR}
    constraintSetup [
     {targetBone {boneName=spine_01}}
     {targetBone {boneName=spine_02}}
     {targetBone {boneName=spine_03}}
     {targetBone {boneName=deform_thigh_fr_l}}
     {targetBone {boneName=pelvis}}
    ]
    constraintWeights [1,1,1,1,1]
   }
   n17 = "Constraint - dyn_belly_medallion_r_MSTR (PPPPP)"!(
    ComponentPose=@n16
    Alpha=1
    ConstraintWeights_0=0.0474
    ConstraintWeights_1:float=0.3852
    ConstraintWeights_2:float=0.1159
    ConstraintWeights_3:float=0.0111
    ConstraintWeights_4:float=0.4404
   ) -> (Pose) {
    alphaBoolBlend {blendOption=Linear}
    boneToModify {boneName=dyn_belly_medallion_r_MSTR}
    constraintSetup [
     {targetBone {boneName=spine_03}}
     {targetBone {boneName=spine_01}}
     {targetBone {boneName=spine_02}}
     {targetBone {boneName=deform_thigh_fr_r}}
     {targetBone {boneName=pelvis}}
    ]
    constraintWeights [1,1,1,1,1]
   }
   n18 = "Constraint - dyn_shldrPad_main_l (PPP)"!(ComponentPose=@n17, Alpha=1, ConstraintWeights_0=0.0386, ConstraintWeights_1:float=0.494, ConstraintWeights_2:float=0.4674) -> (Pose) {
    alphaBoolBlend {blendOption=Linear}
    boneToModify {boneName=dyn_shldrPad_main_l}
    constraintSetup [{targetBone {boneName=spine_03}}, {targetBone {boneName=spine_05}}, {targetBone {boneName=clavicle_l}}]
    constraintWeights [1,1,1]
   }
   n19 = "Constraint - dyn_shldrPad_main_r (PPP)"!(ComponentPose=@n18, Alpha=1, ConstraintWeights_0=0.4674, ConstraintWeights_1:float=0.0386, ConstraintWeights_2:float=0.494) -> (Pose) {
    alphaBoolBlend {blendOption=Linear}
    boneToModify {boneName=dyn_shldrPad_main_r}
    constraintSetup [{targetBone {boneName=clavicle_r}}, {targetBone {boneName=spine_03}}, {targetBone {boneName=spine_05}}]
    constraintWeights [1,1,1]
   }
   n20 = "Constraint - dyn_shldrPad_top_dualOri_l (PP)"!(ComponentPose=@n19, Alpha=1, ConstraintWeights_0=0.6, ConstraintWeights_1:float=0.4) -> (Pose) {
    alphaBoolBlend {blendOption=Linear}
    boneToModify {boneName=dyn_shldrPad_top_dualOri_l}
    constraintSetup [{targetBone {boneName=clavicle_l}}, {targetBone {boneName=upperarm_l}}]
    constraintWeights [1,1]
   }
   n21 = "Constraint - dyn_shldrPad_top_dualOri_r (PP)"!(ComponentPose=@n20, Alpha=1, ConstraintWeights_0=0.6, ConstraintWeights_1:float=0.4) -> (Pose) {
    alphaBoolBlend {blendOption=Linear}
    boneToModify {boneName=dyn_shldrPad_top_dualOri_r}
    constraintSetup [{targetBone {boneName=clavicle_r}}, {targetBone {boneName=upperarm_r}}]
    constraintWeights [1,1]
   }
   n22 = "Constraint - dyn_shldrPad_top_base_l (P)"(ComponentPose=@n21, Alpha=1, ConstraintWeights_0=1) {
    alphaBoolBlend {blendOption=Linear}
    boneToModify {boneName=dyn_shldrPad_top_base_l}
    constraintSetup [{targetBone {boneName=dyn_shldrPad_top_dualOri_l}}]
    constraintWeights [1]
   }
   n23 = "Constraint - dyn_shldrPad_top_base_r (P)"(ComponentPose=@n22, Alpha=1, ConstraintWeights_0=1) {
    alphaBoolBlend {blendOption=Linear}
    boneToModify {boneName=dyn_shldrPad_top_base_r}
    constraintSetup [{targetBone {boneName=dyn_shldrPad_top_dualOri_r}}]
    constraintWeights [1]
   }
   n24 = Look At - Bone: dyn_shldrPad_top_aim_l(ComponentPose=@n23, Alpha=1) {
    alphaBoolBlend {blendOption=Linear}
    boneToModify {boneName=dyn_shldrPad_top_aim_l}
    lookAtLocation {x=0}
    lookAtTarget {boneReference {boneName=lowerarm_l}}
    lookAt_Axis {axis=<1 0>}
    lookUp_Axis {axis=<0 1>}
   }
   n25 = Look At - Bone: dyn_shldrPad_top_aim_r(ComponentPose=@n24, Alpha=1) {
    alphaBoolBlend {blendOption=Linear}
    boneToModify {boneName=dyn_shldrPad_top_aim_r}
    lookAtLocation {x=0}
    lookAtTarget {boneReference {boneName=lowerarm_r}}
    lookAt_Axis {axis=<-1 0>}
    lookUp_Axis {axis=<0 1>}
   }
   n26 = Pose Driver - Source: upperarm_l!(SourcePose:Pose Link Structure=Component To Local(ComponentPose=@n25)) -> (Pose:Pose Link Structure) {
    is=title AnimGraphNode_PoseDriver
    poseAsset=&/M_MED_Historian_Heavy_shoulderPad_poses_PoseAsset
    onlyDriveBones [
     {boneName=dyn_shldrPad_front_pose_l}
     {boneName=dyn_shldrPad_back_pose_l}
     {boneName=dyn_shldrPad_top_pose_l}
     {boneName=dyn_shldrPad_main_pose_l}
    ]
    poseTargets [
     {
      bApplyCustomCurve=false
      bIsHidden=false
      distanceMethod=DefaultMethod
      drivenName=Pose_0
      functionType=DefaultFunction
      targetScale=0.7603
      boneTransforms [{targetRotation=<rot -37.7574 0.1509 2.6875>, targetTranslation=<15.5111 0 0>}]
      customCurve {defaultValue=3.4028234663852886e+38, postInfinityExtrap=RCCE_Constant, preInfinityExtrap=RCCE_Constant, keys []}
      targetRotation=<rot 0 0 0>
     }
     {
      bApplyCustomCurve=false
      bIsHidden=false
      distanceMethod=DefaultMethod
      drivenName=Pose_1
      functionType=DefaultFunction
      targetScale=0.7569
      boneTransforms [{targetRotation=<rot 8.0908 -4.843 3.6512>, targetTranslation=<15.5111 0 0>}]
      customCurve {defaultValue=3.4028234663852886e+38, postInfinityExtrap=RCCE_Constant, preInfinityExtrap=RCCE_Constant, keys []}
      targetRotation=<rot 0 0 0>
     }
     {
      bApplyCustomCurve=false
      bIsHidden=false
      distanceMethod=DefaultMethod
      drivenName=Pose_2
      functionType=DefaultFunction
      targetScale=1
      boneTransforms [{targetRotation=<rot 68.389 3.8186 9.8568>, targetTranslation=<15.5111 0 0>}]
      customCurve {defaultValue=3.4028234663852886e+38, postInfinityExtrap=RCCE_Constant, preInfinityExtrap=RCCE_Constant, keys []}
      targetRotation=<rot 0 0 0>
     }
     {
      bApplyCustomCurve=false
      bIsHidden=false
      distanceMethod=DefaultMethod
      drivenName=Pose_3
      functionType=DefaultFunction
      targetScale=0.6481
      boneTransforms [{targetRotation=<rot -0.7431 65.5344 8.8406>, targetTranslation=<15.5111 0 0>}]
      customCurve {defaultValue=3.4028234663852886e+38, postInfinityExtrap=RCCE_Constant, preInfinityExtrap=RCCE_Constant, keys []}
      targetRotation=<rot 0 0 0>
     }
     {
      bApplyCustomCurve=false
      bIsHidden=false
      distanceMethod=DefaultMethod
      drivenName=Pose_4
      functionType=DefaultFunction
      targetScale=0.6481
      boneTransforms [{targetRotation=<rot -6.1308 104.2528 6.4245>, targetTranslation=<15.5111 0 0>}]
      customCurve {defaultValue=3.4028234663852886e+38, postInfinityExtrap=RCCE_Constant, preInfinityExtrap=RCCE_Constant, keys []}
      targetRotation=<rot 0 0 0>
     }
     {
      bApplyCustomCurve=false
      bIsHidden=false
      distanceMethod=DefaultMethod
      drivenName=Pose_5
      functionType=DefaultFunction
      targetScale=0.7569
      boneTransforms [{targetRotation=<rot 8.2406 -51.1305 -3.2973>, targetTranslation=<15.5111 0 0>}]
      customCurve {defaultValue=3.4028234663852886e+38, postInfinityExtrap=RCCE_Constant, preInfinityExtrap=RCCE_Constant, keys []}
      targetRotation=<rot 0 0 0>
     }
     {
      bApplyCustomCurve=false
      bIsHidden=false
      distanceMethod=DefaultMethod
      drivenName=Pose_6
      functionType=DefaultFunction
      targetScale=0.8313
      boneTransforms [{targetRotation=<rot -87.1937 -48.1479 49.8145>, targetTranslation=<15.5111 0 0>}]
      customCurve {defaultValue=3.4028234663852886e+38, postInfinityExtrap=RCCE_Constant, preInfinityExtrap=RCCE_Constant, keys []}
      targetRotation=<rot 0 0 0>
     }
     {
      bApplyCustomCurve=false
      bIsHidden=false
      distanceMethod=DefaultMethod
      drivenName=Pose_7
      functionType=DefaultFunction
      targetScale=0.9523
      boneTransforms [{targetRotation=<rot 41.2291 91.503 11.0323>, targetTranslation=<15.5111 0 0>}]
      customCurve {defaultValue=3.4028234663852886e+38, postInfinityExtrap=RCCE_Constant, preInfinityExtrap=RCCE_Constant, keys []}
      targetRotation=<rot 0 0 0>
     }
     {
      bApplyCustomCurve=false
      bIsHidden=false
      distanceMethod=DefaultMethod
      drivenName=Pose_8
      functionType=DefaultFunction
      targetScale=0.8864
      boneTransforms [{targetRotation=<rot -40.8085 99.3949 8.4524>, targetTranslation=<15.5111 0 0>}]
      customCurve {defaultValue=3.4028234663852886e+38, postInfinityExtrap=RCCE_Constant, preInfinityExtrap=RCCE_Constant, keys []}
      targetRotation=<rot 0 0 0>
     }
    ]
    rBFParams {radius=60}
    sourceBones [{boneName=upperarm_l}]
   }
   n27 = Pose Driver - Source: upperarm_r!(SourcePose:Pose Link Structure=@n26) -> (Pose:Pose Link Structure) {
    is=title AnimGraphNode_PoseDriver
    poseAsset=&M_MED_Historian_Heavy_shoulderPad_poses_PoseAsset
    onlyDriveBones [
     {boneName=dyn_shldrPad_front_pose_r}
     {boneName=dyn_shldrPad_back_pose_r}
     {boneName=dyn_shldrPad_top_pose_r}
     {boneName=dyn_shldrPad_main_pose_r}
    ]
    poseTargets [
     {
      bApplyCustomCurve=false
      bIsHidden=false
      distanceMethod=DefaultMethod
      drivenName=Pose_0
      functionType=DefaultFunction
      targetScale=0.7603
      boneTransforms [{targetRotation=<rot -37.7574 0.1509 2.6875>, targetTranslation=<-15.5111 0 0>}]
      customCurve {defaultValue=3.4028234663852886e+38, postInfinityExtrap=RCCE_Constant, preInfinityExtrap=RCCE_Constant, keys []}
      targetRotation=<rot 0 0 0>
     }
     {
      bApplyCustomCurve=false
      bIsHidden=false
      distanceMethod=DefaultMethod
      drivenName=Pose_1
      functionType=DefaultFunction
      targetScale=0.7569
      boneTransforms [{targetRotation=<rot 8.0908 -4.843 3.6512>, targetTranslation=<-15.5111 0 0>}]
      customCurve {defaultValue=3.4028234663852886e+38, postInfinityExtrap=RCCE_Constant, preInfinityExtrap=RCCE_Constant, keys []}
      targetRotation=<rot 0 0 0>
     }
     {
      bApplyCustomCurve=false
      bIsHidden=false
      distanceMethod=DefaultMethod
      drivenName=Pose_2
      functionType=DefaultFunction
      targetScale=1
      boneTransforms [{targetRotation=<rot 68.389 3.8186 9.8568>, targetTranslation=<-15.5111 0 0>}]
      customCurve {defaultValue=3.4028234663852886e+38, postInfinityExtrap=RCCE_Constant, preInfinityExtrap=RCCE_Constant, keys []}
      targetRotation=<rot 0 0 0>
     }
     {
      bApplyCustomCurve=false
      bIsHidden=false
      distanceMethod=DefaultMethod
      drivenName=Pose_3
      functionType=DefaultFunction
      targetScale=0.6481
      boneTransforms [{targetRotation=<rot -0.7431 65.5344 8.8406>, targetTranslation=<-15.5111 0 0>}]
      customCurve {defaultValue=3.4028234663852886e+38, postInfinityExtrap=RCCE_Constant, preInfinityExtrap=RCCE_Constant, keys []}
      targetRotation=<rot 0 0 0>
     }
     {
      bApplyCustomCurve=false
      bIsHidden=false
      distanceMethod=DefaultMethod
      drivenName=Pose_4
      functionType=DefaultFunction
      targetScale=0.6481
      boneTransforms [{targetRotation=<rot -6.1308 104.2528 6.4245>, targetTranslation=<-15.5111 0 0>}]
      customCurve {defaultValue=3.4028234663852886e+38, postInfinityExtrap=RCCE_Constant, preInfinityExtrap=RCCE_Constant, keys []}
      targetRotation=<rot 0 0 0>
     }
     {
      bApplyCustomCurve=false
      bIsHidden=false
      distanceMethod=DefaultMethod
      drivenName=Pose_5
      functionType=DefaultFunction
      targetScale=0.7569
      boneTransforms [{targetRotation=<rot 8.2406 -51.1305 -3.2973>, targetTranslation=<-15.5111 0 0>}]
      customCurve {defaultValue=3.4028234663852886e+38, postInfinityExtrap=RCCE_Constant, preInfinityExtrap=RCCE_Constant, keys []}
      targetRotation=<rot 0 0 0>
     }
     {
      bApplyCustomCurve=false
      bIsHidden=false
      distanceMethod=DefaultMethod
      drivenName=Pose_6
      functionType=DefaultFunction
      targetScale=0.8313
      boneTransforms [{targetRotation=<rot -87.1937 -48.1479 49.8145>, targetTranslation=<-15.5111 0 0>}]
      customCurve {defaultValue=3.4028234663852886e+38, postInfinityExtrap=RCCE_Constant, preInfinityExtrap=RCCE_Constant, keys []}
      targetRotation=<rot 0 0 0>
     }
     {
      bApplyCustomCurve=false
      bIsHidden=false
      distanceMethod=DefaultMethod
      drivenName=Pose_7
      functionType=DefaultFunction
      targetScale=0.9523
      boneTransforms [{targetRotation=<rot 41.2291 91.503 11.0323>, targetTranslation=<-15.5111 0 0>}]
      customCurve {defaultValue=3.4028234663852886e+38, postInfinityExtrap=RCCE_Constant, preInfinityExtrap=RCCE_Constant, keys []}
      targetRotation=<rot 0 0 0>
     }
     {
      bApplyCustomCurve=false
      bIsHidden=false
      distanceMethod=DefaultMethod
      drivenName=Pose_8
      functionType=DefaultFunction
      targetScale=0.8864
      boneTransforms [{targetRotation=<rot -40.8085 99.3949 8.4524>, targetTranslation=<-15.5111 0 0>}]
      customCurve {defaultValue=3.4028234663852886e+38, postInfinityExtrap=RCCE_Constant, preInfinityExtrap=RCCE_Constant, keys []}
      targetRotation=<rot 0 0 0>
     }
    ]
    rBFParams {radius=60}
    sourceBones [{boneName=upperarm_r}]
   }
   n28 = "Transform (Modify) Bone - Bone: dyn_shldrPad_front_pose_r"(
    ComponentPose=Local To Component(LocalPose=@n27)
    Alpha=Get pistonAimAlpha_R() -> (pistonAimAlpha_R:double)
    Translation="0,-2,2"
    Rotation="-30,0,50"
    Scale="1,1,1"
   ) {
    rotationMode=BMM_Additive
    rotationSpace=BCS_BoneSpace
    scaleMode=BMM_Additive
    scaleSpace=BCS_BoneSpace
    translationMode=BMM_Additive
    translationSpace=BCS_BoneSpace
    alphaBoolBlend {blendOption=Linear}
    boneToModify {boneName=dyn_shldrPad_front_pose_r}
    rotation {pitch=-30, roll=50}
    translation {y=-2, z=2}
   }
   n29 = "Transform (Modify) Bone - Bone: dyn_shldrPad_main_pose_l"(ComponentPose=@n28, Alpha=Get uparmFwdBk_L() -> (uparmFwdBk_L:double), Translation="0,0,0", Rotation="30,0,-15", Scale="1,1,1") {
    rotationMode=BMM_Additive
    rotationSpace=BCS_BoneSpace
    scaleMode=BMM_Additive
    scaleSpace=BCS_BoneSpace
    translationMode=BMM_Additive
    translationSpace=BCS_BoneSpace
    alphaBoolBlend {blendOption=Linear}
    alphaScaleBiasClamp {bMapRange=true, inRange {max=80, min=60}}
    boneToModify {boneName=dyn_shldrPad_main_pose_l}
    rotation {pitch=30, roll=-15}
   }
   n30 = Get bIsSkyDiving()
   n31 = "Transform (Modify) Bone - Bone: dyn_shldrPad_main_pose_l"!(ComponentPose=@n29, bAlphaBoolEnabled:bool=@n30, Translation="1,0,2", Rotation="0,0,0", Scale="1,1,1") -> (Pose) {
    alphaInputType=Bool
    rotationMode=BMM_Additive
    rotationSpace=BCS_BoneSpace
    scaleMode=BMM_Additive
    scaleSpace=BCS_BoneSpace
    translationMode=BMM_Additive
    translationSpace=BCS_BoneSpace
    alphaBoolBlend {blendOption=Linear}
    boneToModify {boneName=dyn_shldrPad_main_pose_l}
    translation {x=1, z=2}
   }
   n32 = "Transform (Modify) Bone - Bone: dyn_shldrPad_main_pose_r"!(ComponentPose=@n31, bAlphaBoolEnabled:bool=@n30, Translation="-1,0,-2", Rotation="0,0,0", Scale="1,1,1") -> (Pose) {
    alphaInputType=Bool
    rotationMode=BMM_Additive
    rotationSpace=BCS_BoneSpace
    scaleMode=BMM_Additive
    scaleSpace=BCS_BoneSpace
    translationMode=BMM_Additive
    translationSpace=BCS_BoneSpace
    alphaBoolBlend {blendOption=Linear}
    boneToModify {boneName=dyn_shldrPad_main_pose_r}
    translation {x=-1, z=-2}
   }
   n33 = "Transform (Modify) Bone - Bone: dyn_shldrPad_front_pose_l"!(ComponentPose=@n32, bAlphaBoolEnabled:bool=@n30, Translation="0,2,0", Rotation="0,0,0", Scale="1,1,1") -> (Pose) {
    alphaInputType=Bool
    rotationMode=BMM_Additive
    rotationSpace=BCS_BoneSpace
    scaleMode=BMM_Additive
    scaleSpace=BCS_BoneSpace
    translationMode=BMM_Additive
    translationSpace=BCS_BoneSpace
    alphaBoolBlend {blendOption=Linear}
    boneToModify {boneName=dyn_shldrPad_front_pose_l}
    translation {y=2}
   }
   n34 = "Transform (Modify) Bone - Bone: dyn_shldrPad_front_pose_r"!(ComponentPose=@n33, bAlphaBoolEnabled:bool=@n30, Translation="0,-2,0", Rotation="0,0,0", Scale="1,1,1") -> (Pose) {
    alphaInputType=Bool
    rotationMode=BMM_Additive
    rotationSpace=BCS_BoneSpace
    scaleMode=BMM_Additive
    scaleSpace=BCS_BoneSpace
    translationMode=BMM_Additive
    translationSpace=BCS_BoneSpace
    alphaBoolBlend {blendOption=Linear}
    boneToModify {boneName=dyn_shldrPad_front_pose_r}
    translation {y=-2}
   }
   n35 = "Transform (Modify) Bone - Bone: dyn_root_simspace"(ComponentPose=@n34, Alpha=1, Translation="0,0,0", Rotation=Get root_space_rotation() -> (root_space_rotation:Rotator), Scale="1,1,1") {
    rotationMode=BMM_Replace
    rotationSpace=BCS_WorldSpace
    translationSpace=BCS_BoneSpace
    alphaBoolBlend {blendOption=Linear}
    alphaScaleBiasClamp {bInterpResult=true, interpSpeedDecreasing=2, interpSpeedIncreasing=2}
    boneToModify {boneName=dyn_root_simspace}
   }
   n36 = RigidBody(ComponentPose=@n35, Alpha=Get Skydive Loincloth Alpha() -> (Skydive Loincloth Alpha:double), OverrideWorldGravity=Get GravityOverride()) {
    bClampLinearTranslationLimitToRefPose=true
    bOverrideWorldGravity=true
    overridePhysicsAsset=&/M_MED_Historian_Heavy_Dyn_loincloth_perf
    simulationSpace=BaseBoneSpace
    alphaBoolBlend {blendOption=Linear}
    baseBoneRef {boneName=pelvis}
    componentAppliedLinearAccClamp=<3000 3000 3000>
    componentLinearAccScale {z=0.5}
    componentLinearVelScale {z=0.5}
    overrideWorldGravity {z=-980}
    simSpaceSettings {
     maxAngularAcceleration=1000
     maxAngularVelocity=3
     maxLinearAcceleration=250
     maxLinearVelocity=500
     velocityScaleZ=0
     worldAlpha=1
     externalLinearDragV=<0.5 0.5>
    }
   }
   n37 = "Constraint - dyn_strip_r_A_MSTR (P)"!(ComponentPose=@n36, Alpha=1, ConstraintWeights_0=1, ConstraintWeights_1:float=0) -> (Pose) {
    alphaBoolBlend {blendOption=Linear}
    boneToModify {boneName=dyn_strip_r_A_MSTR}
    constraintSetup [{targetBone {boneName=dyn_loincloth_fr_r_01}}]
    constraintWeights [1,0]
   }
   n38 = "Constraint - dyn_strip_r_B_MSTR (PPPPPPP)"!(
    ComponentPose=@n37
    Alpha=1
    ConstraintWeights_0=0.0124
    ConstraintWeights_1:float=0.2314
    ConstraintWeights_2:float=0.4364
    ConstraintWeights_3:float=0.0885
    ConstraintWeights_4:float=0.0419
    ConstraintWeights_5:float=0.113
    ConstraintWeights_6:float=0.0764
   ) -> (Pose) {
    alphaBoolBlend {blendOption=Linear}
    boneToModify {boneName=dyn_strip_r_B_MSTR}
    constraintSetup [
     {targetBone {boneName=deform_thigh_bk_r}}
     {targetBone {boneName=dyn_loincloth_fr_r_01}}
     {targetBone {boneName=dyn_hip_deformation_fr_r}}
     {targetBone {boneName=thigh_twist_01_r}}
     {targetBone {boneName=dyn_loincloth_fr_r_MSTR}}
     {targetBone {boneName=pelvis}}
     {targetBone {boneName=deform_thigh_fr_r}}
    ]
    constraintWeights [1,1,1,1,1,1,1]
   }
   n39 = "Constraint - dyn_strip_r_C_MSTR (PPPPPPP)"!(
    ComponentPose=@n38
    Alpha=1
    ConstraintWeights_0=0.0369
    ConstraintWeights_1:float=0.0356
    ConstraintWeights_2:float=0.3897
    ConstraintWeights_3:float=0.1471
    ConstraintWeights_4:float=0.0208
    ConstraintWeights_5:float=0.2336
    ConstraintWeights_6:float=0.1363
   ) -> (Pose) {
    alphaBoolBlend {blendOption=Linear}
    boneToModify {boneName=dyn_strip_r_C_MSTR}
    constraintSetup [
     {targetBone {boneName=deform_thigh_bk_r}}
     {targetBone {boneName=dyn_loincloth_fr_r_01}}
     {targetBone {boneName=dyn_hip_deformation_fr_r}}
     {targetBone {boneName=thigh_twist_01_r}}
     {targetBone {boneName=spine_01}}
     {targetBone {boneName=pelvis}}
     {targetBone {boneName=deform_thigh_fr_r}}
    ]
    constraintWeights [1,1,1,1,1,1,1]
   }
   n40 = "Constraint - dyn_strip_r_D_MSTR (PPPPPP)"!(
    ComponentPose=@n39
    Alpha=1
    ConstraintWeights_0=0.0991
    ConstraintWeights_1:float=0.0814
    ConstraintWeights_2:float=0.2192
    ConstraintWeights_3:float=0.0315
    ConstraintWeights_4:float=0.4281
    ConstraintWeights_5:float=0.1408
   ) -> (Pose) {
    alphaBoolBlend {blendOption=Linear}
    boneToModify {boneName=dyn_strip_r_D_MSTR}
    constraintSetup [
     {targetBone {boneName=deform_thigh_bk_r}}
     {targetBone {boneName=dyn_hip_deformation_fr_r}}
     {targetBone {boneName=thigh_twist_01_r}}
     {targetBone {boneName=spine_01}}
     {targetBone {boneName=pelvis}}
     {targetBone {boneName=deform_thigh_fr_r}}
    ]
    constraintWeights [1,1,1,1,1,1]
   }
   n41 = "Constraint - dyn_strip_r_E_MSTR (PPPPPP)"!(
    ComponentPose=@n40
    Alpha=1
    ConstraintWeights_0=0.1531
    ConstraintWeights_1:float=0.0167
    ConstraintWeights_2:float=0.218
    ConstraintWeights_3:float=0.0285
    ConstraintWeights_4:float=0.5117
    ConstraintWeights_5:float=0.0719
   ) -> (Pose) {
    alphaBoolBlend {blendOption=Linear}
    boneToModify {boneName=dyn_strip_r_E_MSTR}
    constraintSetup [
     {targetBone {boneName=deform_thigh_bk_r}}
     {targetBone {boneName=dyn_loincloth_bk_r_01}}
     {targetBone {boneName=thigh_twist_01_r}}
     {targetBone {boneName=spine_01}}
     {targetBone {boneName=pelvis}}
     {targetBone {boneName=deform_thigh_fr_r}}
    ]
    constraintWeights [1,1,1,1,1,1]
   }
   n42 = "Constraint - dyn_strip_r_F_MSTR (PPPPP)"!(
    ComponentPose=@n41
    Alpha=1
    ConstraintWeights_0=0.5094
    ConstraintWeights_1:float=0.0762
    ConstraintWeights_2:float=0.0233
    ConstraintWeights_3:float=0.0437
    ConstraintWeights_4:float=0.3474
   ) -> (Pose) {
    alphaBoolBlend {blendOption=Linear}
    boneToModify {boneName=dyn_strip_r_F_MSTR}
    constraintSetup [
     {targetBone {boneName=pelvis}}
     {targetBone {boneName=deform_thigh_bk_r}}
     {targetBone {boneName=dyn_loincloth_bk_l_02}}
     {targetBone {boneName=thigh_twist_01_r}}
     {targetBone {boneName=dyn_loincloth_bk_r_01}}
    ]
    constraintWeights [1,1,1,1,1]
   }
   n43 = "Constraint - dyn_strip_r_G_MSTR (PP)"!(ComponentPose=@n42, Alpha=1, ConstraintWeights_0=0.2431, ConstraintWeights_1:float=0.7569) -> (Pose) {
    alphaBoolBlend {blendOption=Linear}
    boneToModify {boneName=dyn_strip_r_G_MSTR}
    constraintSetup [{targetBone {boneName=dyn_loincloth_bk_r_02}}, {targetBone {boneName=dyn_loincloth_bk_r_01}}]
    constraintWeights [1,1]
   }
   n44 = "Constraint - dyn_strip_r_H_MSTR (PPP)"!(ComponentPose=@n43, Alpha=1, ConstraintWeights_0=0.9546, ConstraintWeights_1:float=0.0219, ConstraintWeights_2:float=0.0235) -> (Pose) {
    alphaBoolBlend {blendOption=Linear}
    boneToModify {boneName=dyn_strip_r_H_MSTR}
    constraintSetup [{targetBone {boneName=dyn_loincloth_bk_r_02}}, {targetBone {boneName=dyn_loincloth_bk_r_01}}, {targetBone {boneName=dyn_loincloth_bk_03}}]
    constraintWeights [1,1,1]
   }
   n45 = "Constraint - dyn_strip_l_K_MSTR (PPP)"!(ComponentPose=@n44, Alpha=1, ConstraintWeights_0=0.9156, ConstraintWeights_1:float=0.0297, ConstraintWeights_2:float=0.0547) -> (Pose) {
    alphaBoolBlend {blendOption=Linear}
    boneToModify {boneName=dyn_strip_l_K_MSTR}
    constraintSetup [{targetBone {boneName=dyn_loincloth_bk_l_01}}, {targetBone {boneName=pelvis}}, {targetBone {boneName=dyn_loincloth_bk_l_02}}]
    constraintWeights [1,1,1]
   }
   n46 = "Constraint - dyn_strip_l_L_MSTR (PPPPPPPP)"!(
    ComponentPose=@n45
    Alpha=1
    ConstraintWeights_0=0.0265
    ConstraintWeights_1:float=0.0106
    ConstraintWeights_2:float=0.0874
    ConstraintWeights_3:float=0.8069
    ConstraintWeights_4:float=0.0126
    ConstraintWeights_5:float=0.012
    ConstraintWeights_6:float=0.0332
    ConstraintWeights_7:float=0.0108
   ) -> (Pose) {
    alphaBoolBlend {blendOption=Linear}
    boneToModify {boneName=dyn_strip_l_L_MSTR}
    constraintSetup [
     {targetBone {boneName=dyn_loincloth_bk_r_02}}
     {targetBone {boneName=dyn_loincloth_bk_r_01}}
     {targetBone {boneName=dyn_loincloth_bk_l_01}}
     {targetBone {boneName=dyn_loincloth_bk_l_02}}
     {targetBone {boneName=dyn_loincloth_bk_04}}
     {targetBone {boneName=dyn_loincloth_bk_02}}
     {targetBone {boneName=dyn_loincloth_bk_03}}
     {targetBone {boneName=pelvis}}
    ]
    constraintWeights [1,1,1,1,1,1,1,1]
   }
   n47 = "Constraint - dyn_strip_l_B_MSTR (PPPPP)"!(
    ComponentPose=@n46
    Alpha=1
    ConstraintWeights_0=0.0145
    ConstraintWeights_1:float=0.0353
    ConstraintWeights_2:float=0.0162
    ConstraintWeights_3:float=0.8306
    ConstraintWeights_4:float=0.1034
   ) -> (Pose) {
    alphaBoolBlend {blendOption=Linear}
    boneToModify {boneName=dyn_strip_l_B_MSTR}
    constraintSetup [
     {targetBone {boneName=dyn_loincloth_fr_03}}
     {targetBone {boneName=dyn_loincloth_fr_02}}
     {targetBone {boneName=dyn_loincloth_fr_01}}
     {targetBone {boneName=dyn_loincloth_fr_l_02}}
     {targetBone {boneName=dyn_loincloth_fr_l_01}}
    ]
    constraintWeights [1,1,1,1,1]
   }
   n48 = "Constraint - dyn_strip_l_C_MSTR (PP)"!(ComponentPose=@n47, Alpha=1, ConstraintWeights_0=0.8721, ConstraintWeights_1:float=0.1279) -> (Pose) {
    alphaBoolBlend {blendOption=Linear}
    boneToModify {boneName=dyn_strip_l_C_MSTR}
    constraintSetup [{targetBone {boneName=dyn_loincloth_fr_l_02}}, {targetBone {boneName=dyn_loincloth_fr_l_01}}]
    constraintWeights [1,1]
   }
   n49 = "Constraint - dyn_strip_l_D_MSTR (P)"!(ComponentPose=@n48, Alpha=1, ConstraintWeights_0=1, ConstraintWeights_1:float=0) -> (Pose) {
    alphaBoolBlend {blendOption=Linear}
    boneToModify {boneName=dyn_strip_l_D_MSTR}
    constraintSetup [{targetBone {boneName=dyn_loincloth_fr_l_01}}]
    constraintWeights [1,0]
   }
   n50 = "Constraint - dyn_strip_l_E_MSTR (PPPPPPP)"!(
    ComponentPose=@n49
    Alpha=1
    ConstraintWeights_0=0.0844
    ConstraintWeights_1:float=0.0212
    ConstraintWeights_2:float=0.279
    ConstraintWeights_3:float=0.0835
    ConstraintWeights_4:float=0.3808
    ConstraintWeights_5:float=0.1362
    ConstraintWeights_6:float=0.015
   ) -> (Pose) {
    alphaBoolBlend {blendOption=Linear}
    boneToModify {boneName=dyn_strip_l_E_MSTR}
    constraintSetup [
     {targetBone {boneName=deform_thigh_fr_l}}
     {targetBone {boneName=dyn_loincloth_fr_l_MSTR}}
     {targetBone {boneName=dyn_loincloth_fr_l_01}}
     {targetBone {boneName=thigh_twist_01_l}}
     {targetBone {boneName=dyn_hip_deformation_fr_l}}
     {targetBone {boneName=pelvis}}
     {targetBone {boneName=deform_thigh_bk_l}}
    ]
    constraintWeights [1,1,1,1,1,1,1]
   }
   n51 = "Constraint - dyn_strip_l_F_MSTR (PPPPPPP)"!(
    ComponentPose=@n50
    Alpha=1
    ConstraintWeights_0=0.1231
    ConstraintWeights_1:float=0.0615
    ConstraintWeights_2:float=0.0169
    ConstraintWeights_3:float=0.1321
    ConstraintWeights_4:float=0.4097
    ConstraintWeights_5:float=0.2243
    ConstraintWeights_6:float=0.0324
   ) -> (Pose) {
    alphaBoolBlend {blendOption=Linear}
    boneToModify {boneName=dyn_strip_l_F_MSTR}
    constraintSetup [
     {targetBone {boneName=deform_thigh_fr_l}}
     {targetBone {boneName=dyn_loincloth_fr_l_01}}
     {targetBone {boneName=spine_01}}
     {targetBone {boneName=thigh_twist_01_l}}
     {targetBone {boneName=dyn_hip_deformation_fr_l}}
     {targetBone {boneName=pelvis}}
     {targetBone {boneName=deform_thigh_bk_l}}
    ]
    constraintWeights [1,1,1,1,1,1,1]
   }
   n52 = "Constraint - dyn_strip_l_G_MSTR (PPPPPP)"!(
    ComponentPose=@n51
    Alpha=1
    ConstraintWeights_0=0.1408
    ConstraintWeights_1:float=0.0315
    ConstraintWeights_2:float=0.2192
    ConstraintWeights_3:float=0.0814
    ConstraintWeights_4:float=0.4281
    ConstraintWeights_5:float=0.0991
   ) -> (Pose) {
    alphaBoolBlend {blendOption=Linear}
    boneToModify {boneName=dyn_strip_l_G_MSTR}
    constraintSetup [
     {targetBone {boneName=deform_thigh_fr_l}}
     {targetBone {boneName=spine_01}}
     {targetBone {boneName=thigh_twist_01_l}}
     {targetBone {boneName=dyn_hip_deformation_fr_l}}
     {targetBone {boneName=pelvis}}
     {targetBone {boneName=deform_thigh_bk_l}}
    ]
    constraintWeights [1,1,1,1,1,1]
   }
   n53 = "Constraint - dyn_strip_l_H_MSTR (PPPPPP)"!(
    ComponentPose=@n52
    Alpha=1
    ConstraintWeights_0=0.0719
    ConstraintWeights_1:float=0.0167
    ConstraintWeights_2:float=0.0285
    ConstraintWeights_3:float=0.218
    ConstraintWeights_4:float=0.5117
    ConstraintWeights_5:float=0.1531
   ) -> (Pose) {
    alphaBoolBlend {blendOption=Linear}
    boneToModify {boneName=dyn_strip_l_H_MSTR}
    constraintSetup [
     {targetBone {boneName=deform_thigh_fr_l}}
     {targetBone {boneName=dyn_loincloth_bk_l_01}}
     {targetBone {boneName=spine_01}}
     {targetBone {boneName=thigh_twist_01_l}}
     {targetBone {boneName=pelvis}}
     {targetBone {boneName=deform_thigh_bk_l}}
    ]
    constraintWeights [1,1,1,1,1,1]
   }
   n54 = "Constraint - dyn_strip_l_J_MSTR (PPPP)"!(
    ComponentPose=@n53
    Alpha=1
    ConstraintWeights_0=0.7356
    ConstraintWeights_1:float=0.1122
    ConstraintWeights_2:float=0.1193
    ConstraintWeights_3:float=0.0329
   ) -> (Pose) {
    alphaBoolBlend {blendOption=Linear}
    boneToModify {boneName=dyn_strip_l_J_MSTR}
    constraintSetup [
     {targetBone {boneName=pelvis}}
     {targetBone {boneName=thigh_twist_01_l}}
     {targetBone {boneName=deform_thigh_bk_l}}
     {targetBone {boneName=deform_thigh_fr_l}}
    ]
    constraintWeights [1,1,1,1]
   }
   n55 = Anim Dynamics - Bone: dyn_omega_01_MSTR(ComponentPose=@n54, Alpha=1, ExternalForce="0,0,0") {
    angularDampingOverride=0.995
    bGravityOverrideInSimSpace=true
    bOverrideAngularDamping=true
    bUseGravityOverride=true
    simulationSpace=RootRelative
    alphaBoolBlend {blendOption=Linear}
    alphaScaleBiasClamp {bInterpResult=true}
    boundBone {boneName=dyn_omega_01_MSTR}
    componentAppliedLinearAccClamp=<3000 3000 3000>
    componentLinearAccScale=<0.25 0.25 0.3>
    componentLinearVelScale=<0.125 0.125 0.15>
    gravityOverride {z=-850}
    physicsBodyDefinitions [{
      sphereCollisionRadius=0
      boundBone {boneName=dyn_omega_01_MSTR}
      boxExtents=<15 15 2>
      constraintSetup {angularLimitsMax=<5 1 15>, angularLimitsMin=<-10 -75 -20>, angularTarget=<0 0 0>}
      localJointOffset=<3 0 0>
     }]
    relativeSpaceBone {boneName=spine_05}
   }
   n56 = Anim Dynamics - Bone: dyn_knife_l_01(ComponentPose=@n55, Alpha=1, ExternalForce="0,0,0") {
    angularDampingOverride=0.999
    bGravityOverrideInSimSpace=true
    bOverrideAngularDamping=true
    bUseGravityOverride=true
    simulationSpace=RootRelative
    alphaBoolBlend {blendOption=Linear}
    alphaScaleBiasClamp {bInterpResult=true}
    boundBone {boneName=dyn_knife_l_01}
    chainEnd {boneName=dyn_knife_l_02}
    componentAppliedLinearAccClamp=<3000 3000 3000>
    componentLinearAccScale=<0.2 0.2 0.3>
    componentLinearVelScale=<0.1 0.1 0.125>
    gravityOverride {z=-980}
    physicsBodyDefinitions [{
      sphereCollisionRadius=0
      boundBone {boneName=dyn_knife_l_01}
      boxExtents=<25 15 10>
      constraintSetup {angularLimitsMax=<5 0 5>, angularLimitsMin=<-2 -15 -5>, angularTarget=<0 0 0>}
      localJointOffset=<3 0 0>
     }]
    relativeSpaceBone {boneName=pelvis}
   }
   n57 = Anim Dynamics - Bone: dyn_knife_l_02(ComponentPose=@n56, Alpha=1, ExternalForce="0,0,0") {
    angularDampingOverride=0.999
    bGravityOverrideInSimSpace=true
    bOverrideAngularDamping=true
    bUseGravityOverride=true
    simulationSpace=RootRelative
    alphaBoolBlend {blendOption=Linear}
    alphaScaleBiasClamp {bInterpResult=true}
    boundBone {boneName=dyn_knife_l_02}
    chainEnd {boneName=dyn_knife_l_02}
    componentAppliedLinearAccClamp=<3000 3000 3000>
    componentLinearAccScale=<0.4 0.4 0.4>
    componentLinearVelScale=<0.2 0.2 0.2>
    gravityOverride {z=-980}
    physicsBodyDefinitions [{
      sphereCollisionRadius=0
      boundBone {boneName=dyn_knife_l_02}
      boxExtents=<25 15 10>
      constraintSetup {angularLimitsMax=<5 0 5>, angularLimitsMin=<-2 -15 -5>, angularTarget=<0 0 0>}
      localJointOffset=<3 0 0>
     }]
    relativeSpaceBone {boneName=dyn_knife_l_01}
   }
   n58 = Anim Dynamics - Bone: dyn_satchel_l_01(ComponentPose=@n57, Alpha=1, ExternalForce="0,0,0") {
    angularDampingOverride=0.999
    bGravityOverrideInSimSpace=true
    bOverrideAngularDamping=true
    bUseGravityOverride=true
    simulationSpace=RootRelative
    alphaBoolBlend {blendOption=Linear}
    alphaScaleBiasClamp {bInterpResult=true}
    boundBone {boneName=dyn_satchel_l_01}
    chainEnd {boneName=dyn_satchel_l_02}
    componentAppliedLinearAccClamp=<3000 3000 3000>
    componentLinearAccScale=<0.2 0.2 0.35>
    componentLinearVelScale=<0.1 0.1 0.125>
    gravityOverride {z=-980}
    physicsBodyDefinitions [{
      sphereCollisionRadius=0
      boundBone {boneName=dyn_satchel_l_01}
      boxExtents=<15 25 10>
      constraintSetup {angularLimitsMax=<5 0 2>, angularLimitsMin=<-5 -35 -2>, angularTarget=<0 0 0>}
      localJointOffset=<3 0 0>
     }]
    relativeSpaceBone {boneName=pelvis}
   }
   n59 = Anim Dynamics - Bone: dyn_satchel_l_02(ComponentPose=@n58, Alpha=1, ExternalForce="0,0,0") {
    angularDampingOverride=0.999
    bGravityOverrideInSimSpace=true
    bOverrideAngularDamping=true
    bUseGravityOverride=true
    simulationSpace=RootRelative
    alphaBoolBlend {blendOption=Linear}
    alphaScaleBiasClamp {bInterpResult=true}
    boundBone {boneName=dyn_satchel_l_02}
    chainEnd {boneName=dyn_satchel_l_02}
    componentAppliedLinearAccClamp=<3000 3000 3000>
    componentLinearAccScale=<0.4 0.4 0.4>
    componentLinearVelScale=<0.2 0.2 0.2>
    gravityOverride {z=-980}
    physicsBodyDefinitions [{
      sphereCollisionRadius=0
      boundBone {boneName=dyn_satchel_l_02}
      boxExtents=<15 25 10>
      constraintSetup {angularLimitsMax=<5 0 2>, angularLimitsMin=<-5 -35 -2>, angularTarget=<0 0 0>}
      localJointOffset=<3 0 0>
     }]
    relativeSpaceBone {boneName=dyn_satchel_l_01}
   }
   n60 = Anim Dynamics - Bone: dyn_bag_l_01(ComponentPose=@n59, Alpha=1, ExternalForce="0,0,0") {
    angularDampingOverride=0.999
    bGravityOverrideInSimSpace=true
    bOverrideAngularDamping=true
    bUseGravityOverride=true
    simulationSpace=RootRelative
    alphaBoolBlend {blendOption=Linear}
    alphaScaleBiasClamp {bInterpResult=true}
    boundBone {boneName=dyn_bag_l_01}
    chainEnd {boneName=dyn_bag_l_03}
    componentAppliedLinearAccClamp=<3000 3000 3000>
    componentLinearAccScale=<0.2 0.2 0.3>
    componentLinearVelScale=<0.1 0.1 0.125>
    gravityOverride {z=-980}
    physicsBodyDefinitions [{
      sphereCollisionRadius=0
      boundBone {boneName=dyn_bag_l_01}
      boxExtents=<25 25 10>
      constraintSetup {angularLimitsMax=<5 0 5>, angularLimitsMin=<-5 -25 -5>, angularTarget=<0 0 0>}
      localJointOffset=<3 0 0>
     }]
    relativeSpaceBone {boneName=pelvis}
   }
   n61 = Anim Dynamics - Bone: dyn_bag_l_02(ComponentPose=@n60, Alpha=1, ExternalForce="0,0,0") {
    angularDampingOverride=0.999
    bGravityOverrideInSimSpace=true
    bOverrideAngularDamping=true
    bUseGravityOverride=true
    simulationSpace=RootRelative
    alphaBoolBlend {blendOption=Linear}
    alphaScaleBiasClamp {bInterpResult=true}
    boundBone {boneName=dyn_bag_l_02}
    chainEnd {boneName=dyn_bag_l_03}
    componentAppliedLinearAccClamp=<3000 3000 3000>
    componentLinearAccScale=<0.4 0.4 0.4>
    componentLinearVelScale=<0.2 0.2 0.2>
    gravityOverride {z=-980}
    physicsBodyDefinitions [{
      sphereCollisionRadius=0
      boundBone {boneName=dyn_bag_l_02}
      boxExtents=<25 25 10>
      constraintSetup {angularLimitsMax=<5 0 5>, angularLimitsMin=<-5 -25 -5>, angularTarget=<0 0 0>}
      localJointOffset=<3 0 0>
     }]
    relativeSpaceBone {boneName=dyn_bag_l_01}
   }
   n62 = Anim Dynamics - Bone: dyn_bag_l_03(ComponentPose=@n61, Alpha=1, ExternalForce="0,0,0") {
    angularDampingOverride=0.999
    bGravityOverrideInSimSpace=true
    bOverrideAngularDamping=true
    bUseGravityOverride=true
    simulationSpace=RootRelative
    alphaBoolBlend {blendOption=Linear}
    alphaScaleBiasClamp {bInterpResult=true}
    boundBone {boneName=dyn_bag_l_03}
    chainEnd {boneName=dyn_bag_l_03}
    componentAppliedLinearAccClamp=<3000 3000 3000>
    componentLinearAccScale=<0.4 0.4 0.4>
    componentLinearVelScale=<0.2 0.2 0.2>
    gravityOverride {z=-980}
    physicsBodyDefinitions [{
      sphereCollisionRadius=0
      boundBone {boneName=dyn_bag_l_03}
      boxExtents=<25 25 10>
      constraintSetup {angularLimitsMax=<5 0 5>, angularLimitsMin=<-5 -25 -5>, angularTarget=<0 0 0>}
      localJointOffset=<3 0 0>
     }]
    relativeSpaceBone {boneName=dyn_bag_l_02}
   }
   n63 = Anim Dynamics - Bone: dyn_belt_tips_top_01(ComponentPose=@n62, Alpha=1, ExternalForce="0,0,0") {
    angularDampingOverride=0.998
    bGravityOverrideInSimSpace=true
    bOverrideAngularDamping=true
    bUseGravityOverride=true
    simulationSpace=RootRelative
    alphaBoolBlend {blendOption=Linear}
    alphaScaleBiasClamp {bInterpResult=true}
    boundBone {boneName=dyn_belt_tips_top_01}
    componentAppliedLinearAccClamp=<3000 3000 3000>
    componentLinearAccScale=<0.3 0.3 0.333>
    componentLinearVelScale=<0.3 0.3 0.3>
    gravityOverride {z=-400}
    physicsBodyDefinitions [{
      sphereCollisionRadius=0
      boundBone {boneName=dyn_belt_tips_top_01}
      boxExtents=<8 3 8>
      constraintSetup {angularLimitsMax=<10 5 5>, angularLimitsMin=<-10 -45 -5>, angularTarget=<0 0 0>}
      localJointOffset=<1 0 0>
     }]
    relativeSpaceBone {boneName=pelvis}
   }
   n64 = Anim Dynamics - Bone: dyn_belt_tips_bott_01(ComponentPose=@n63, Alpha=1, ExternalForce="0,0,0") {
    angularDampingOverride=0.998
    bGravityOverrideInSimSpace=true
    bOverrideAngularDamping=true
    bUseGravityOverride=true
    simulationSpace=RootRelative
    alphaBoolBlend {blendOption=Linear}
    alphaScaleBiasClamp {bInterpResult=true}
    boundBone {boneName=dyn_belt_tips_bott_01}
    componentAppliedLinearAccClamp=<3000 3000 3000>
    componentLinearAccScale=<0.3 0.3 0.333>
    componentLinearVelScale=<0.3 0.3 0.3>
    gravityOverride {z=-400}
    physicsBodyDefinitions [{
      sphereCollisionRadius=0
      boundBone {boneName=dyn_belt_tips_bott_01}
      boxExtents=<8 3 8>
      constraintSetup {angularLimitsMax=<10 5 5>, angularLimitsMin=<-10 -45 -5>, angularTarget=<0 0 0>}
      localJointOffset=<1 0 0>
     }]
    relativeSpaceBone {boneName=pelvis}
   }
   n65 = Anim Dynamics - Bone: dyn_belly_medallion_r_01(ComponentPose=@n64, Alpha=1, ExternalForce="0,0,0") {
    angularDampingOverride=0.998
    bGravityOverrideInSimSpace=true
    bOverrideAngularDamping=true
    bUseGravityOverride=true
    simulationSpace=RootRelative
    alphaBoolBlend {blendOption=Linear}
    alphaScaleBiasClamp {bInterpResult=true}
    boundBone {boneName=dyn_belly_medallion_r_01}
    componentAppliedLinearAccClamp=<3000 3000 3000>
    componentLinearAccScale=<0.3 0.3 0.333>
    componentLinearVelScale=<0.3 0.3 0.3>
    gravityOverride {z=-850}
    physicsBodyDefinitions [{
      sphereCollisionRadius=0
      boundBone {boneName=dyn_belly_medallion_r_01}
      boxExtents=<12 12 3>
      constraintSetup {angularLimitsMax=<5 1 10>, angularLimitsMin=<-5 -75 -10>, angularTarget=<0 0 0>}
      localJointOffset=<1 0 0>
     }]
    relativeSpaceBone {boneName=pelvis}
   }
   n66 = Anim Dynamics - Bone: dyn_belly_medallion_l_01(ComponentPose=@n65, Alpha=1, ExternalForce="0,0,0") {
    angularDampingOverride=0.998
    bGravityOverrideInSimSpace=true
    bOverrideAngularDamping=true
    bUseGravityOverride=true
    simulationSpace=RootRelative
    alphaBoolBlend {blendOption=Linear}
    alphaScaleBiasClamp {bInterpResult=true}
    boundBone {boneName=dyn_belly_medallion_l_01}
    componentAppliedLinearAccClamp=<3000 3000 3000>
    componentLinearAccScale=<0.3 0.3 0.333>
    componentLinearVelScale=<0.3 0.3 0.3>
    gravityOverride {z=-850}
    physicsBodyDefinitions [{
      sphereCollisionRadius=0
      boundBone {boneName=dyn_belly_medallion_l_01}
      boxExtents=<15 15 4>
      constraintSetup {angularLimitsMax=<5 1 10>, angularLimitsMin=<-5 -75 -10>, angularTarget=<0 0 0>}
      localJointOffset=<1 0 0>
     }]
    relativeSpaceBone {boneName=pelvis}
   }
   n67 = "Constraint - dyn_belt_knot_MSTR (PPPPP)"!(
    ComponentPose=@n66
    Alpha=0.8
    ConstraintWeights_0=0.8242
    ConstraintWeights_1:float=0.0163
    ConstraintWeights_2:float=0.0228
    ConstraintWeights_3:float=0.0989
    ConstraintWeights_4:float=0.0378
   ) -> (Pose) {
    alphaBoolBlend {blendOption=Linear}
    boneToModify {boneName=dyn_belt_knot_MSTR}
    constraintSetup [
     {targetBone {boneName=dyn_loincloth_fr_r_MSTR}}
     {targetBone {boneName=dyn_loincloth_fr_01}}
     {targetBone {boneName=dyn_loincloth_fr_r_02}}
     {targetBone {boneName=dyn_loincloth_fr_r_01}}
     {targetBone {boneName=dyn_loincloth_fr_02}}
    ]
    constraintWeights [1,1,1,1,1]
   }
   n68 = "Constraint - dyn_belt_strap_fr_01 (P)"(ComponentPose=@n67, Alpha=0.6, ConstraintWeights_0=1) {
    alphaBoolBlend {blendOption=Linear}
    boneToModify {boneName=dyn_belt_strap_fr_01}
    constraintSetup [{targetBone {boneName=dyn_loincloth_fr_r_01}}]
    constraintWeights [1]
   }
   n69 = Get Skydive Belt Alpha() -> (Skydive Belt Alpha:double)
   n70 = "Constraint - dyn_satchel_l_MSTR (P)"(ComponentPose=@n68, Alpha=@n69, ConstraintWeights_0=1) {
    alphaBoolBlend {blendOption=Linear}
    boneToModify {boneName=dyn_satchel_l_MSTR}
    constraintSetup [{targetBone {boneName=dyn_loincloth_bk_01}}]
    constraintWeights [1]
   }
   n71 = "Transform (Modify) Bone - Bone: dyn_knife_l_01"(ComponentPose=@n70, Alpha=@n69, Translation="-2,2,4", Rotation="0.000,0.000,30", Scale="1,1,1") {
    rotationMode=BMM_Additive
    rotationSpace=BCS_BoneSpace
    translationMode=BMM_Additive
    translationSpace=BCS_BoneSpace
    alphaBoolBlend {blendOption=Linear}
    boneToModify {boneName=dyn_knife_l_01}
    rotation {roll=30}
    translation=<-2 2 4>
   }
   n72 = "Transform (Modify) Bone - Bone: dyn_loincloth_fr_r_01"(ComponentPose=@n71, Alpha=1, Translation="0,0,0.5", Rotation="0,0,0", Scale="1,1,1") {
    rotationMode=BMM_Additive
    rotationSpace=BCS_BoneSpace
    translationMode=BMM_Additive
    translationSpace=BCS_BoneSpace
    alphaBoolBlend {blendOption=Linear}
    boneToModify {boneName=dyn_loincloth_fr_r_01}
    translation {z=0.5}
   }
   n73 = "Constraint - dyn_loincloth_fr_r_MSTR (P)"(ComponentPose=@n72, Alpha=0.1, ConstraintWeights_0=1) {
    alphaBoolBlend {blendOption=Linear}
    boneToModify {boneName=dyn_loincloth_fr_r_MSTR}
    constraintSetup [{targetBone {boneName=thigh_r}}]
    constraintWeights [1]
   }
   n74 = "Constraint - dyn_loincloth_fr_l_MSTR (P)"(ComponentPose=@n73, Alpha=0.1, ConstraintWeights_0=1) {
    alphaBoolBlend {blendOption=Linear}
    boneToModify {boneName=dyn_loincloth_fr_l_MSTR}
    constraintSetup [{targetBone {boneName=thigh_l}}]
    constraintWeights [1]
   }
   n75 = "Constraint - dyn_loincloth_fr_l_01 (P)"(ComponentPose=@n74, Alpha=0.6, ConstraintWeights_0=1) {
    alphaBoolBlend {blendOption=Linear}
    boneToModify {boneName=dyn_loincloth_fr_l_01}
    constraintSetup [{targetBone {boneName=thigh_l}}]
    constraintWeights [1]
   }
   n76 = "Constraint - dyn_strip_r_A_MSTR (PP)"!(ComponentPose=@n75, Alpha=0.5, ConstraintWeights_0=0.5, ConstraintWeights_1:float=0.5) -> (Pose) {
    alphaBoolBlend {blendOption=Linear}
    boneToModify {boneName=dyn_strip_r_A_MSTR}
    constraintSetup [{targetBone {boneName=dyn_loincloth_fr_r_01}}, {targetBone {boneName=thigh_r}}]
    constraintWeights [1,1]
   }
   n77 = "Constraint - dyn_strip_r_B_MSTR (PP)"!(ComponentPose=@n76, Alpha=0.5, ConstraintWeights_0=0.3, ConstraintWeights_1:float=0.7) -> (Pose) {
    alphaBoolBlend {blendOption=Linear}
    boneToModify {boneName=dyn_strip_r_B_MSTR}
    constraintSetup [{targetBone {boneName=dyn_loincloth_fr_r_01}}, {targetBone {boneName=thigh_r}}]
    constraintWeights [1,1]
   }
   n78 = "Constraint - dyn_strip_r_A_MSTR (PP)"(ComponentPose=@n77, Alpha=@n4, ConstraintWeights_0=0.5, ConstraintWeights_1=0.5) {
    alphaBoolBlend {blendOption=Linear}
    boneToModify {boneName=dyn_strip_r_A_MSTR}
    constraintSetup [{targetBone {boneName=dyn_loincloth_fr_r_01}}, {targetBone {boneName=thigh_r}}]
    constraintWeights [1,1]
   }
   n79 = "Constraint - dyn_strip_r_C_MSTR (P)"(ComponentPose=@n78, Alpha=@n4, ConstraintWeights_0=1) {
    alphaBoolBlend {blendOption=Linear}
    boneToModify {boneName=dyn_strip_r_C_MSTR}
    constraintSetup [{targetBone {boneName=thigh_r}}]
    constraintWeights [1]
   }
   n80 = "Constraint - dyn_strip_l_B_MSTR (P)"(ComponentPose=@n79, Alpha=@n4, ConstraintWeights_0=1) {
    alphaBoolBlend {blendOption=Linear}
    boneToModify {boneName=dyn_strip_l_B_MSTR}
    constraintSetup [{targetBone {boneName=thigh_l}}]
    constraintWeights [1]
   }
   n81 = "Constraint - dyn_strip_l_C_MSTR (P)"(ComponentPose=@n80, Alpha=@n4, ConstraintWeights_0=1) {
    alphaBoolBlend {blendOption=Linear}
    boneToModify {boneName=dyn_strip_l_C_MSTR}
    constraintSetup [{targetBone {boneName=thigh_l}}]
    constraintWeights [1]
   }
   n82 = "Constraint - dyn_strip_l_D_MSTR (P)"!(ComponentPose=@n81, Alpha=@n4, ConstraintWeights_0=1) -> (Pose) {
    alphaBoolBlend {blendOption=Linear}
    boneToModify {boneName=dyn_strip_l_D_MSTR}
    constraintSetup [{targetBone {boneName=thigh_l}}]
    constraintWeights [1]
   }
   n83 = Control Rig(Source=Component To Local(ComponentPose=@n82)) {
    alphaBoolBlend {blendOption=Linear}
    controlRigAssetReference {blueprintRigClass=&/M_MED_Historian_Heavy_Deform_CtrlRig_side_strips.M_MED_Historian_Heavy_Deform_CtrlRig_side_strips_C}
   }
   n84 = RigidBody(ComponentPose=Local To Component(LocalPose=@n83), Alpha=Get Skydive Loincloth Alpha(), OverrideWorldGravity=@n6) {
    bClampLinearTranslationLimitToRefPose=true
    bOverrideWorldGravity=true
    overridePhysicsAsset=&/M_MED_Historian_Heavy_Dyn_strips_perf
    simulationSpace=BaseBoneSpace
    alphaBoolBlend {blendOption=Linear}
    baseBoneRef {boneName=pelvis}
    componentAppliedLinearAccClamp=<3000 3000 3000>
    componentLinearAccScale {z=0.5}
    componentLinearVelScale {z=0.5}
    overrideWorldGravity {z=-980}
    simSpaceSettings {
     maxAngularAcceleration=1000
     maxAngularVelocity=3
     maxLinearAcceleration=250
     maxLinearVelocity=500
     velocityScaleZ=0
     worldAlpha=1
     externalLinearDragV=<0.1 0.1>
    }
   }
   n85 = "Constraint - dyn_strip_l_B_01 (PP)"!(ComponentPose=@n84, Alpha=0.7, ConstraintWeights_0=0.7, ConstraintWeights_1:float=0.3) -> (Pose) {
    alphaBoolBlend {blendOption=Linear}
    boneToModify {boneName=dyn_strip_l_B_01}
    constraintSetup [{targetBone {boneName=dyn_loincloth_fr_l_01}}, {targetBone {boneName=thigh_l}}]
    constraintWeights [1,1]
   }
   n86 = "Constraint - dyn_strip_l_D_01 (PP)"!(ComponentPose=@n85, Alpha=0.7, ConstraintWeights_0=0.3, ConstraintWeights_1:float=0.7) -> (Pose) {
    alphaBoolBlend {blendOption=Linear}
    boneToModify {boneName=dyn_strip_l_D_01}
    constraintSetup [{targetBone {boneName=dyn_loincloth_fr_l_01}}, {targetBone {boneName=thigh_l}}]
    constraintWeights [1,1]
   }
   n87 = "Transform (Modify) Bone - Bone: dyn_bag_l_MSTR"!(ComponentPose=@n86, bAlphaBoolEnabled:bool=True, Translation="0,0,0", Rotation="-5,0,0", Scale="1,1,1") -> (Pose) {
    alphaInputType=Bool
    bAlphaBoolEnabled=false
    rotationMode=BMM_Additive
    rotationSpace=BCS_BoneSpace
    translationMode=BMM_Additive
    translationSpace=BCS_BoneSpace
    alphaBoolBlend {blendOption=Linear}
    boneToModify {boneName=dyn_bag_l_MSTR}
    rotation {pitch=-5}
   }
   n88 = "Transform (Modify) Bone - Bone: dyn_bag_l_03"!(ComponentPose=@n87, bAlphaBoolEnabled:bool=True, Translation="0,0,0.25", Rotation="20,0,0", Scale="1,1,1") -> (Pose) {
    alphaInputType=Bool
    bAlphaBoolEnabled=false
    rotationMode=BMM_Additive
    rotationSpace=BCS_BoneSpace
    translationMode=BMM_Additive
    translationSpace=BCS_BoneSpace
    alphaBoolBlend {blendOption=Linear}
    boneToModify {boneName=dyn_bag_l_03}
    rotation {pitch=20}
    translation {z=0.25}
   }
   n89 = Get crouch alpha() -> (crouch alpha:double)
   n90 = "Transform (Modify) Bone - Bone: dyn_strip_l_E_MSTR"(ComponentPose=@n88, Alpha=@n89, Translation="0,0,0", Rotation="15,-10,0", Scale="1,1,1") {
    rotationMode=BMM_Additive
    rotationSpace=BCS_BoneSpace
    translationMode=BMM_Additive
    translationSpace=BCS_BoneSpace
    alphaBoolBlend {blendOption=Linear}
    boneToModify {boneName=dyn_strip_l_E_MSTR}
    rotation {pitch=15, yaw=-10}
   }
   n91 = "Constraint - dyn_strip_l_D_MSTR (P)"(ComponentPose=@n90, Alpha=@n89, ConstraintWeights_0=1) {
    alphaBoolBlend {blendOption=Linear}
    boneToModify {boneName=dyn_strip_l_D_MSTR}
    constraintSetup [{targetBone {boneName=dyn_loincloth_fr_l_01}}]
    constraintWeights [1]
   }
   n92 = "Constraint - dyn_strip_l_D_01 (P)"(ComponentPose=@n91, Alpha=@n89, ConstraintWeights_0=1) {
    alphaBoolBlend {blendOption=Linear}
    boneToModify {boneName=dyn_strip_l_D_01}
    constraintSetup [{targetBone {boneName=dyn_loincloth_fr_l_01}}]
    constraintWeights [1]
   }
   n93 = "Constraint - dyn_strip_l_B_MSTR (T)"(ComponentPose=@n92, Alpha=@n89, ConstraintWeights_0=1) {
    alphaBoolBlend {blendOption=Linear}
    boneToModify {boneName=dyn_strip_l_B_MSTR}
    constraintSetup [{transformType=Translation, targetBone {boneName=dyn_loincloth_fr_03}}]
    constraintWeights [1]
   }
   n94 = "Constraint - dyn_strip_l_B_01 (T)"(ComponentPose=@n93, Alpha=@n89, ConstraintWeights_0=1) {
    alphaBoolBlend {blendOption=Linear}
    boneToModify {boneName=dyn_strip_l_B_01}
    constraintSetup [{transformType=Translation, targetBone {boneName=dyn_loincloth_fr_03}}]
    constraintWeights [1]
   }
   n95 = "Constraint - dyn_strip_l_B_02 (T)"(ComponentPose=@n94, Alpha=@n89, ConstraintWeights_0=1) {
    alphaBoolBlend {blendOption=Linear}
    boneToModify {boneName=dyn_strip_l_B_02}
    constraintSetup [{transformType=Translation, targetBone {boneName=dyn_loincloth_fr_03}}]
    constraintWeights [1]
   }
   Output Pose(Result=Component To Local(ComponentPose=@n95)) {layerGroup=None}
  ]}
 Complex Sim Space {nodes [
   n1 = vector + vector(
    A:Vector=GetSocketLocation!(self:Scene Component Object Reference=Get Skel_Mesh(), InSocketName:Name=root) -> (ReturnValue:Vector)
    B:Vector="0, 0,165"
   ) -> (ReturnValue:Vector) {is=title K2Node_CommutativeAssociativeBinaryOperator}
   n2 = Complex Sim Space!() -> (then, InterpSpeedMax:double=~, B:double=~, DeltaTime:double=~, ErrorTolerance:double=~) {is=title K2Node_FunctionEntry}
   n3 = Get target_sim_space() -> (target_sim_space:Rotator)
   n4 = Get root_space_rotation()
   n5 = Set root_space_rotation(
    execute=[@n20.then, @n21.then]
    root_space_rotation:Rotator=RInterpTo!(Current:Rotator=@n4, Target:Rotator=@n3, DeltaTime:float=@n2.DeltaTime, InterpSpeed:float=Get InterpSpeed() -> (InterpSpeed:double)) -> (ReturnValue:Rotator)
   ) -> (Output_Get:Rotator)
   n6 = DrawDebugCoordinateSystem!(
    execute
    AxisLoc:Vector=@n1
    AxisRot:Rotator=@n5
    Scale:float=35
    Duration:float=0
    Thickness:float=3
    DepthPriority:EDrawDebugSceneDepthPriorityGroup Enum=World
   ) -> (then)
   Set inComponentSpaceLastFrame(execute=[@n6.then, @n5.then], inComponentSpaceLastFrame:bool=Get inComponentSpace() -> (inComponentSpace:bool)) -> (Output_Get:bool)
   > Branch(Condition="Equal (Rotator)"!(A:Rotator=@n4, B:Rotator=@n3, ErrorTolerance:float=@n2.ErrorTolerance) -> (ReturnValue:bool))
   > Set lockedOnSimSpace(lockedOnSimSpace:bool=true) -> (Output_Get:bool)
   > n7 = Set InterpSpeed(InterpSpeed:double=0) -> (Output_Get:double)
   DrawDebugCoordinateSystem(AxisLoc=@n1, AxisRot=Get root_rotation() -> (root_rotation:Rotator), Scale=75, Duration=0, Thickness=1.5, DepthPriority=World)
   PrintString!(execute, InString="To String (Float)"(InDouble=@n7), bPrintToScreen, bPrintToLog, TextColor, Duration, Key) -> (then)
   n8 = GetOwningActor()
   n9 = Get Actor Location(self=@n8)
   n10 = vector * float(A=GetActorForwardVector!(self:Actor Object Reference=@n8) -> (ReturnValue:Vector), B=500)
   DrawDebugLine!(
    execute
    LineStart:Vector=vector + vector(A=@n9, B="0,0,87")
    LineEnd:Vector=vector + vector(A=vector + vector(A=@n9, B=@n10), B="0,0,87")
    LineColor:Linear Color Structure=(R=0.6348,G=0,B=0.6406,A=1)
    Duration:float=0
    Thickness:float=0.5
    DepthPriority:EDrawDebugSceneDepthPriorityGroup Enum=World
   ) -> (then)
   Reroute Node(InputPin=@n16.then_2)
   Reroute Node(InputPin=@n16.then_1)
   Set root_rotation(
    execute=@n16.then_0
    root_rotation:Rotator=GetSocketRotation!(self:Scene Component Object Reference=Get Skel_Mesh(), InSocketName:Name=root) -> (ReturnValue:Rotator)
   ) -> (Output_Get:Rotator)
   n11 = GetOwningActor()
   n12 = "Equal (Boolean)"!(A:bool=Get inComponentSpaceLastFrame() -> (inComponentSpaceLastFrame:bool), B:bool=Get inComponentSpace()) -> (ReturnValue:bool)
   n13 = Get Actor Rotation!(self:Actor Object Reference=@n11) -> (ReturnValue:Rotator) {is=title K2Node_CallFunction}
   n14 = "float > float"(A=Get PawnSpeed2D(), B=@n2.B) {is=title K2Node_CallFunction}
   n15 = Combine Rotators!(A:Rotator=@n13, B:Rotator="0,-90, 0") -> (ReturnValue:Rotator)
   n16 = Sequence!(execute=@n2.then) -> (then_0, then_1, then_2:Exec, then_3:Exec)
   Set InterpSpeedMax(execute=@n16.then_3, InterpSpeedMax:double=@n2) -> (Output_Get:double)
   > n17 = Set inComponentSpace(inComponentSpace:bool=@n14) -> (Output_Get:bool)
   > n18 = Branch(Condition=@n17)
   > n19 = Set target_sim_space(target_sim_space:Rotator=@n15) -> (Output_Get:Rotator)
   n20 = Branch(execute=[@n22.then, @n19.then], Condition=@n12)
   Set lockedOnSimSpace(execute=@n20.else, lockedOnSimSpace=false)
   > n21 = Set InterpSpeed(InterpSpeed=Get InterpSpeedMax() -> (InterpSpeedMax:double))
   n22 = Set target_sim_space(execute=@n18.else, target_sim_space=Get root_rotation())
  ]}
 EventGraph {nodes [
   n1 = Event BlueprintUpdateAnimation()
   > n2 = Sequence()
   > n3 = Branch(Condition=Get bIsSwimSprinting() -> (bIsSwimSprinting:bool))
   n4 = PerlinNoise1D!(Value:float=GetGameTimeInSeconds()) -> (ReturnValue:float)
   n5 = float * float(A="Sin (Degrees)"!(A:double=float * float(A=@n4, B=800)) -> (ReturnValue:double), B=-400)
   n6 = GetSocketRotation(self=Get Skel_Mesh(), InSocketName=pelvis)
   n7 = vector + vector(
    A=vector * float(A=GetForwardVector!(InRot:Rotator=@n6) -> (ReturnValue:Vector), B=-300)
    B=MakeVector(Y=0, Z=float + float(A=@n5, B=200) {is=title K2Node_CommutativeAssociativeBinaryOperator})
   )
   > Set GravityOverride(GravityOverride=@n7)
   n8 = Branch(execute=@n3.else, Condition=Get bIsSkyDiving())
   > n9 = Branch(Condition=Get bIsParachuteOpened())
   Set GravityOverride(execute=[@n8.else, @n9.then], GravityOverride="0, 0,-980")
   n10 = Branch(execute=@n9.else, Condition="float > float"(A=Get PawnVelocityZ(), B=50))
   > Set GravityOverride(GravityOverride="0, 0,-100")
   Set GravityOverride(execute=@n10.else, GravityOverride=@n7)
   Complex Sim Space!(execute=@n2.then_1, self, InterpSpeedMax:double=0, B:double=0, DeltaTime:double=@n1.DeltaTimeX, ErrorTolerance:double=0) -> (then) {is=title K2Node_CallFunction}
   n11 = Sequence!(execute=@n2.then_3) -> (then_0, then_1, then_2, then_3, then_4:Exec, then_5:Exec)
   n12 = GetOwnerSkeletalMeshComponent()
   n13 = GetDeltaTransformFromRefPose(self=@n12, BoneName=upperarm_r, BaseName=clavicle_r)
   Set uparmFwdBk_R(execute=@n11.then_1, uparmFwdBk_R:double=BreakRotator(InRot=BreakTransform(InTransform=@n13).Rotation).Yaw) -> (Output_Get:double)
   n14 = Branch(execute=@n11.then_2, Condition="Equal Exactly (String)"(A=Get weapon() -> (weapon:String), B=Pistol))
   > Set pistonAimAlpha_R(pistonAimAlpha_R:double=NormalizeToRange!(Value:double=Get uparmFwdBk_R() -> (uparmFwdBk_R:double), RangeMin:double=80, RangeMax:double=120) -> (ReturnValue:double)) -> (Output_Get:double)
   Set pistonAimAlpha_R(execute=@n14.else, pistonAimAlpha_R=0)
   n15 = Branch(execute=@n11.then_4, Condition=OR Boolean(A=Get bIsJumping(), B=Get bIsFalling()))
   > Set Jump Alpha(Jump Alpha:double=0.8) -> (Output_Get:double)
   Set Jump Alpha(execute=@n15.else, Jump Alpha=0)
   Set Skydive Belt Alpha(execute=@n11.then_3, Skydive Belt Alpha:double=SelectFloat(A=0.5, B=0, bPickA=Get bIsSkyDiving())) -> (Output_Get:double)
   > Set Skydive Loincloth Alpha(Skydive Loincloth Alpha:double=SelectFloat(A=0.5, B=1, bPickA=Get bIsSkydiveFloating())) -> (Output_Get:double)
   n16 = GetDeltaTransformFromRefPose(self=@n12, BoneName=upperarm_l, BaseName=clavicle_l)
   Set uparmFwdBk_L(execute=@n11.then_0, uparmFwdBk_L:double=BreakRotator(InRot=BreakTransform(InTransform=@n16).Rotation).Yaw) -> (Output_Get:double)
   n17 = Cast To FortPlayerAnimInstance(execute=@n2.then_2, Object=GetAnimInstance(self=GetOwnerSkeletalMeshComponent()))
   > Set weapon(weapon:String=Enum to String!(Enumerator:EFortWeaponCoreAnimation Enum=Get WeaponCoreAnim(self=@n17)) -> (ReturnValue:String)) -> (Output_Get:String)
   Event BlueprintInitializeAnimation()
   > Set Skel_Mesh(Skel_Mesh=GetOwnerSkeletalMeshComponent())
   > Set InterpSpeed(InterpSpeed=Get InterpSpeedMax())
   Event OnEndCrouching()
   > Set crouch alpha(crouch alpha:double=0) -> (Output_Get:double)
   Event OnBeginCrouching()
   > Set crouch alpha(crouch alpha=1)
  ]}
]
variables [
 crouch alpha {type=number}
 GravityOverride {type=Vector}
 Idle Standing Alpha {type=number}
 inComponentSpace {type=boolean}
 inComponentSpaceLastFrame {type=boolean}
 InterpSpeed {type=number}
 InterpSpeedMax {type=number}
 Jump Alpha {type=number}
 lockedOnSimSpace {type=boolean}
 pistonAimAlpha_R {type=number}
 root_rotation {type=Rotator}
 root_space_rotation {type=Rotator}
 Skel_Mesh {type=SkeletalMeshComponent}
 Skydive Belt Alpha {type=number}
 Skydive Loincloth Alpha {type=number}
 target_sim_space {type=Rotator}
 uparmFwdBk_L {type=number}
 uparmFwdBk_R {type=number}
 weapon {type=string, default=""}
]
