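# Physics asset of cape colliders for the Historian Heavy medium male player cosmetic, apparently for keeping the cape from clipping into the body.
# Capsules cover the pelvis, spine, neck and clavicles, with doubled or tripled shapes on parts of the spine and a sphere on the head.
# Extra capsules on dynamic bag, knife, satchel and back loincloth bones cover those accessories.
# Arm, leg and foot bodies have no shapes, and no constraints link any bodies.
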PhysicsAsset /BRCosmetics/Characters/Player/Male/Medium/Bodies/M_MED_Historian_Heavy/Meshes/M_MED_Historian_Heavy_CapeColliders
bodies {
 calf_l []
 calf_r []
 clavicle_l [{shapeType=Capsule, length=7.7636, radius=11.1427, center=<9.9642 -0.7741 -1.5923>, rotation=<rot 75.9352 -0.0001 175.5622>}]
 clavicle_r [{shapeType=Capsule, length=7.7636, radius=11.1427, center=<-9.9642 0.7741 1.5923>, rotation=<rot 75.9351 -0.0001 -4.4378>}]
 dyn_bag_l_03 [{
   shapeType=Capsule
   length=17.25
   radius=7.25
   shapeName=dyn_bag_l_03_capsule
   center=<1.1663 0 0.2057>
   rotation=<rot 80 -179.9998 -179.9999>
  }]
 dyn_knife_l_02 [{
   shapeType=Capsule
   length=23.25
   radius=7.5
   shapeName=dyn_knife_l_02_capsule
   center=<2.0071 -0.718 -0.022>
   rotation=<rot -86.5442 80.1574 -99.8591>
  }]
 dyn_loincloth_bk_04 [{shapeType=Capsule, length=13, radius=17, shapeName=dyn_loincloth_bk_04_capsule, center=<0.6866 0 -9.7766>, rotation=<rot -0.0002 0 -90>}]
 dyn_satchel_l_02 [{shapeType=Capsule, length=6.5, radius=8.5, shapeName=dyn_satchel_l_02_capsule, center=<0 -1.6027 -2.3488>, rotation=<rot -0.0002 0 -100>}]
 foot_l []
 foot_r []
 head [{shapeType=Sphere, radius=17.3125, center=<11.028 -0.174 0>}]
 lowerarm_l []
 lowerarm_r []
 neck_02 [{shapeType=Capsule, length=13.4375, radius=7.3175, center=<-0.2642 -0.3408 0>, rotation=<rot 90 -171.1193 -175.3978>}]
 pelvis [{shapeType=Capsule, length=8.9062, radius=15.0388, center=<-2.6215 -1.3695 0>}]
 spine_01 [{shapeType=Capsule, length=10.7187, radius=12.4138, center=<-2.6215 -0.7156 0>}]
 spine_02 [
  {shapeType=Capsule, length=8.8437, radius=12.695, center=<-2.6215 0.4993 0.0001>}
  {shapeType=Capsule, length=29.875, radius=11.1013, center=<2.5024 0.4992 3.9825>, rotation=<rot 90 0.0049 0.0049>}
  {shapeType=Capsule, length=29.875, radius=11.1013, center=<2.4704 0.4991 -3.9825>, rotation=<rot 90 -0.0396 -0.0396>}
 ]
 spine_03 [{shapeType=Capsule, length=10.6562, radius=13.2575, center=<1.2805 0.1364 0.0001>}]
 spine_04 [{shapeType=Capsule, length=12.4062, radius=15.8825, center=<2.1467 -1.2748 0.0002>}]
 spine_05 [
  {shapeType=Capsule, length=18.3437, radius=16.695, center=<-3.9735 -0.9394 0>}
  {shapeType=Capsule, length=11.5312, radius=18.3125, center=<0.7875 -2.7391 0>}
 ]
 thigh_l []
 thigh_r []
 upperarm_l []
 upperarm_r []
}
constraints []
